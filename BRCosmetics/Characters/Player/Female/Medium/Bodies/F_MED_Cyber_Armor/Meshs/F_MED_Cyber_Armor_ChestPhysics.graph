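# Physics asset for the chest straps on the female medium Cyber Armor cosmetic body.
# Left and right straps are capsule chains, each with a small Main root on spine_05 and six segments linked by constraints with 45-degree angular limits.
# Larger capsules on the pelvis, spine and thighs give body volumes around the straps.
# Constraints between body bones and at both strap roots leave swing and twist free.
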PhysicsAsset /BRCosmetics/Characters/Player/Female/Medium/Bodies/F_MED_Cyber_Armor/Meshs/F_MED_Cyber_Armor_ChestPhysics
bodies {
 l_chestStrap_01 [{shapeType=Capsule, length=5.6599, radius=1.0966, shapeName=l_chestStrap_01_capsule, center=<5.5068 0 -0.0001>, rotation=<rot 90 0 0>}]
 l_chestStrap_02 [{shapeType=Capsule, length=7.7355, radius=1.4615, shapeName=l_chestStrap_02_capsule, center=<4.5749 0 -0.0001>, rotation=<rot 90 0 0>}]
 l_chestStrap_03 [{shapeType=Capsule, length=6.8548, radius=2.2053, shapeName=l_chestStrap_03_capsule, center=<4.7733 0 0>, rotation=<rot 90 0 0>}]
 l_chestStrap_04 [{shapeType=Capsule, length=6.6949, radius=2.2053, shapeName=l_chestStrap_04_capsule, center=<4.7828 0 0>, rotation=<rot 90 0 0>}]
 l_chestStrap_05 [{shapeType=Capsule, length=4.5512, radius=2.2053, shapeName=l_chestStrap_05_capsule, center=<3.717 0 0>, rotation=<rot 90 0 0>}]
 l_chestStrap_06 [{shapeType=Capsule, length=4.5512, radius=2.2053, shapeName=l_chestStrap_06_capsule, center=<3.4336 0 -0.0001>, rotation=<rot 90 0 0>}]
 l_chestStrap_Main [{shapeType=Capsule, length=0.505, radius=0.505, shapeName=l_chestStrap_Main_capsule}]
 pelvis [{shapeType=Capsule, length=8.918, radius=10, shapeName=pelvis_capsule, center=<-0.6893 -2.1977 0>}]
 r_chestStrap_01 [{shapeType=Capsule, length=5.6599, radius=1.0966, shapeName=r_chestStrap_01_capsule, center=<5.3242 0 0>, rotation=<rot 90 0 0>}]
 r_chestStrap_02 [{shapeType=Capsule, length=7.3454, radius=1.4615, shapeName=r_chestStrap_02_capsule, center=<4.6565 0 0>, rotation=<rot 90 0 0>}]
 r_chestStrap_03 [{shapeType=Capsule, length=6.8818, radius=2.2053, shapeName=r_chestStrap_03_capsule, center=<4.7444 0 0.0001>, rotation=<rot 90 0 0>}]
 r_chestStrap_04 [{shapeType=Capsule, length=6.6778, radius=2.2053, shapeName=r_chestStrap_04_capsule, center=<4.6208 0 0>, rotation=<rot 90 0 0>}]
 r_chestStrap_05 [{shapeType=Capsule, length=4.5512, radius=2.2053, shapeName=r_chestStrap_05_capsule, center=<3.7152 0 0.0001>, rotation=<rot 90 0 0>}]
 r_chestStrap_06 [{shapeType=Capsule, length=4.5512, radius=2.2053, shapeName=r_chestStrap_06_capsule, center=<3.1925 0 0>, rotation=<rot 90 0 0>}]
 r_chestStrap_Main [{shapeType=Capsule, length=0.505, radius=0.505, shapeName=r_chestStrap_Main_capsule}]
 spine_01 [{shapeType=Capsule, length=3.178, radius=10, shapeName=spine_01_capsule}]
 spine_02 [{shapeType=Capsule, length=1.7785, radius=9.5, shapeName=spine_02_capsule}]
 spine_03 [{shapeType=Capsule, length=3.4884, radius=9.5, shapeName=spine_03_capsule}]
 spine_04 [{shapeType=Capsule, length=3.8712, radius=8.4541, shapeName=spine_04_capsule, center=<1.2342 3.3337 0>}]
 spine_05 [{shapeType=Capsule, length=8.3269, radius=7.6845, shapeName=spine_05_capsule}]
 thigh_l [{shapeType=Capsule, length=20.195, radius=8.28, shapeName=thigh_l_capsule, center=<-17.7142 0 0>, rotation=<rot 90 0 0>}]
 thigh_r [{shapeType=Capsule, length=21.6135, radius=8.2089, shapeName=thigh_r_capsule, center=<18.7042 -0.0001 0>, rotation=<rot 90 0 0>}]
}
constraints [
 {
  bone1Name=spine_01
  bone2Name=pelvis
  swing1LimitDegrees=45
  swing1Motion=Free
  swing2LimitDegrees=45
  swing2Motion=Free
  twistLimitDegrees=45
  twistMotion=Free
 }
 {
  bone1Name=spine_02
  bone2Name=spine_01
  swing1LimitDegrees=45
  swing1Motion=Free
  swing2LimitDegrees=45
  swing2Motion=Free
  twistLimitDegrees=45
  twistMotion=Free
 }
 {
  bone1Name=spine_03
  bone2Name=spine_02
  swing1LimitDegrees=45
  swing1Motion=Free
  swing2LimitDegrees=45
  swing2Motion=Free
  twistLimitDegrees=45
  twistMotion=Free
 }
 {
  bone1Name=spine_04
  bone2Name=spine_03
  swing1LimitDegrees=45
  swing1Motion=Free
  swing2LimitDegrees=45
  swing2Motion=Free
  twistLimitDegrees=45
  twistMotion=Free
 }
 {
  bone1Name=spine_05
  bone2Name=spine_04
  swing1LimitDegrees=45
  swing1Motion=Free
  swing2LimitDegrees=45
  swing2Motion=Free
  twistLimitDegrees=45
  twistMotion=Free
 }
 {
  bone1Name=l_chestStrap_Main
  bone2Name=spine_05
  swing1LimitDegrees=45
  swing1Motion=Free
  swing2LimitDegrees=45
  swing2Motion=Free
  twistLimitDegrees=45
  twistMotion=Free
 }
 {
  bone1Name=r_chestStrap_Main
  bone2Name=spine_05
  swing1LimitDegrees=45
  swing1Motion=Free
  swing2LimitDegrees=45
  swing2Motion=Free
  twistLimitDegrees=45
  twistMotion=Free
 }
 {bone1Name=l_chestStrap_01, bone2Name=l_chestStrap_Main, swing1LimitDegrees=45, swing2LimitDegrees=45, twistLimitDegrees=45}
 {bone1Name=l_chestStrap_02, bone2Name=l_chestStrap_01, swing1LimitDegrees=45, swing2LimitDegrees=45, twistLimitDegrees=45}
 {bone1Name=l_chestStrap_03, bone2Name=l_chestStrap_02, swing1LimitDegrees=45, swing2LimitDegrees=45, twistLimitDegrees=45}
 {bone1Name=l_chestStrap_04, bone2Name=l_chestStrap_03, swing1LimitDegrees=45, swing2LimitDegrees=45, twistLimitDegrees=45}
 {bone1Name=l_chestStrap_05, bone2Name=l_chestStrap_04, swing1LimitDegrees=45, swing2LimitDegrees=45, twistLimitDegrees=45}
 {bone1Name=l_chestStrap_06, bone2Name=l_chestStrap_05, swing1LimitDegrees=45, swing2LimitDegrees=45, twistLimitDegrees=45}
 {bone1Name=r_chestStrap_01, bone2Name=r_chestStrap_Main, swing1LimitDegrees=45, swing2LimitDegrees=45, twistLimitDegrees=45}
 {bone1Name=r_chestStrap_02, bone2Name=r_chestStrap_01, swing1LimitDegrees=45, swing2LimitDegrees=45, twistLimitDegrees=45}
 {bone1Name=r_chestStrap_03, bone2Name=r_chestStrap_02, swing1LimitDegrees=45, swing2LimitDegrees=45, twistLimitDegrees=45}
 {bone1Name=r_chestStrap_04, bone2Name=r_chestStrap_03, swing1LimitDegrees=45, swing2LimitDegrees=45, twistLimitDegrees=45}
 {bone1Name=r_chestStrap_05, bone2Name=r_chestStrap_04, swing1LimitDegrees=45, swing2LimitDegrees=45, twistLimitDegrees=45}
 {bone1Name=r_chestStrap_06, bone2Name=r_chestStrap_05, swing1LimitDegrees=45, swing2LimitDegrees=45, twistLimitDegrees=45}
 {
  bone1Name=thigh_l
  bone2Name=pelvis
  swing1LimitDegrees=45
  swing1Motion=Free
  swing2LimitDegrees=45
  swing2Motion=Free
  twistLimitDegrees=45
  twistMotion=Free
 }
 {
  bone1Name=thigh_r
  bone2Name=pelvis
  swing1LimitDegrees=45
  swing1Motion=Free
  swing2LimitDegrees=45
  swing2Motion=Free
  twistLimitDegrees=45
  twistMotion=Free
 }
]
