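# Physics asset for the female medium Cyber Armor body cosmetic, giving capsule bodies to the character and its two chest straps.
# Larger capsules cover the pelvis, spine, clavicles, arms and thighs, joined by constraints with free swing and twist.
# Each strap is a six-segment capsule chain from a small Main anchor, itself freely constrained to its clavicle.
# Unlike the body joints, links within the strap chains are not set to free rotation.
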PhysicsAsset /BRCosmetics/Characters/Player/Female/Medium/Bodies/F_MED_Cyber_Armor/Meshs/F_MED_Cyber_Armor_FE_Physics
bodies {
 clavicle_l [{shapeType=Capsule, length=9.7207, radius=4.3163, shapeName=clavicle_l_capsule, center=<5.0524 0 0>, rotation=<rot 90 0 0>}]
 clavicle_r [{shapeType=Capsule, length=9.847, radius=4.3656, shapeName=clavicle_r_capsule, center=<-6.5484 0 0>, rotation=<rot 90 0 0>}]
 l_chestStrap_01 [{shapeType=Capsule, length=5.6599, radius=1.0966, shapeName=l_chestStrap_01_capsule, center=<5.5068 0 -0.0001>, rotation=<rot 90 0 0>}]
 l_chestStrap_02 [{shapeType=Capsule, length=7.7355, radius=1.4615, shapeName=l_chestStrap_02_capsule, center=<4.5749 0 -0.0001>, rotation=<rot 90 0 0>}]
 l_chestStrap_03 [{shapeType=Capsule, length=6.8548, radius=2.2053, shapeName=l_chestStrap_03_capsule, center=<4.7733 0 0>, rotation=<rot 90 0 0>}]
 l_chestStrap_04 [{shapeType=Capsule, length=6.6949, radius=2.2053, shapeName=l_chestStrap_04_capsule, center=<4.7828 0 0>, rotation=<rot 90 0 0>}]
 l_chestStrap_05 [{shapeType=Capsule, length=4.5512, radius=2.2053, shapeName=l_chestStrap_05_capsule, center=<3.717 0 0>, rotation=<rot 90 0 0>}]
 l_chestStrap_06 [{shapeType=Capsule, length=4.5512, radius=2.2053, shapeName=l_chestStrap_06_capsule, center=<3.4336 0 -0.0001>, rotation=<rot 90 0 0>}]
 l_chestStrap_Main [{shapeType=Capsule, length=0.505, radius=0.505, shapeName=l_chestStrap_Main_capsule}]
 lowerarm_l [{shapeType=Capsule, length=25.493, radius=4.507, shapeName=lowerarm_l_capsule, center=<16.8907 0 0>, rotation=<rot 90 0 0>}]
 lowerarm_r [{shapeType=Capsule, length=25.493, radius=4.507, shapeName=lowerarm_r_capsule, center=<-16.4953 0 0>, rotation=<rot 90 0 0>}]
 pelvis [{shapeType=Capsule, length=8.918, radius=10, shapeName=pelvis_capsule, center=<-0.6893 -2.1977 0>}]
 r_chestStrap_01 [{shapeType=Capsule, length=5.6599, radius=1.0966, shapeName=r_chestStrap_01_capsule, center=<5.3242 0 0>, rotation=<rot 90 0 0>}]
 r_chestStrap_02 [{shapeType=Capsule, length=7.3454, radius=1.4615, shapeName=r_chestStrap_02_capsule, center=<4.6565 0 0>, rotation=<rot 90 0 0>}]
 r_chestStrap_03 [{shapeType=Capsule, length=6.8818, radius=2.2053, shapeName=r_chestStrap_03_capsule, center=<4.7444 0 0.0001>, rotation=<rot 90 0 0>}]
 r_chestStrap_04 [{shapeType=Capsule, length=6.6778, radius=2.2053, shapeName=r_chestStrap_04_capsule, center=<4.6208 0 0>, rotation=<rot 90 0 0>}]
 r_chestStrap_05 [{shapeType=Capsule, length=4.5512, radius=2.2053, shapeName=r_chestStrap_05_capsule, center=<3.7152 0 0.0001>, rotation=<rot 90 0 0>}]
 r_chestStrap_06 [{shapeType=Capsule, length=4.5512, radius=2.2053, shapeName=r_chestStrap_06_capsule, center=<3.1925 0 0>, rotation=<rot 90 0 0>}]
 r_chestStrap_Main [{shapeType=Capsule, length=0.505, radius=0.505, shapeName=r_chestStrap_Main_capsule}]
 spine_01 [{shapeType=Capsule, length=3.178, radius=10, shapeName=spine_01_capsule}]
 spine_02 [{shapeType=Capsule, length=1.7785, radius=9.5, shapeName=spine_02_capsule}]
 spine_03 [{shapeType=Capsule, length=3.4884, radius=9.5, shapeName=spine_03_capsule}]
 spine_04 [{shapeType=Capsule, length=3.8712, radius=8.4541, shapeName=spine_04_capsule, center=<1.2342 3.3337 0>}]
 spine_05 [{shapeType=Capsule, length=8.3269, radius=7.6845, shapeName=spine_05_capsule}]
 thigh_l [{shapeType=Capsule, length=20.7632, radius=7.7118, shapeName=thigh_l_capsule, center=<-17.7142 0 0>, rotation=<rot 90 0 0>}]
 thigh_r [{shapeType=Capsule, length=22.1817, radius=7.6407, shapeName=thigh_r_capsule, center=<18.7042 -0.0001 0>, rotation=<rot 90 0 0>}]
 upperarm_l [{shapeType=Capsule, length=18.7833, radius=5.1768, shapeName=upperarm_l_capsule, center=<13.6922 0 0>, rotation=<rot 90 0 0>}]
 upperarm_r [{shapeType=Capsule, length=18.9863, radius=5.1768, shapeName=upperarm_r_capsule, center=<-13.5495 0 0>, rotation=<rot 90 0 0>}]
}
constraints [
 {
  bone1Name=spine_01
  bone2Name=pelvis
  swing1LimitDegrees=45
  swing1Motion=Free
  swing2LimitDegrees=45
  swing2Motion=Free
  twistLimitDegrees=45
  twistMotion=Free
 }
 {
  bone1Name=spine_02
  bone2Name=spine_01
  swing1LimitDegrees=45
  swing1Motion=Free
  swing2LimitDegrees=45
  swing2Motion=Free
  twistLimitDegrees=45
  twistMotion=Free
 }
 {
  bone1Name=spine_03
  bone2Name=spine_02
  swing1LimitDegrees=45
  swing1Motion=Free
  swing2LimitDegrees=45
  swing2Motion=Free
  twistLimitDegrees=45
  twistMotion=Free
 }
 {
  bone1Name=spine_04
  bone2Name=spine_03
  swing1LimitDegrees=45
  swing1Motion=Free
  swing2LimitDegrees=45
  swing2Motion=Free
  twistLimitDegrees=45
  twistMotion=Free
 }
 {
  bone1Name=spine_05
  bone2Name=spine_04
  swing1LimitDegrees=45
  swing1Motion=Free
  swing2LimitDegrees=45
  swing2Motion=Free
  twistLimitDegrees=45
  twistMotion=Free
 }
 {
  bone1Name=l_chestStrap_Main
  bone2Name=clavicle_l
  swing1LimitDegrees=45
  swing1Motion=Free
  swing2LimitDegrees=45
  swing2Motion=Free
  twistLimitDegrees=45
  twistMotion=Free
 }
 {
  bone1Name=r_chestStrap_Main
  bone2Name=clavicle_r
  swing1LimitDegrees=45
  swing1Motion=Free
  swing2LimitDegrees=45
  swing2Motion=Free
  twistLimitDegrees=45
  twistMotion=Free
 }
 {bone1Name=l_chestStrap_01, bone2Name=l_chestStrap_Main, swing1LimitDegrees=45, swing2LimitDegrees=45, twistLimitDegrees=45}
 {bone1Name=l_chestStrap_02, bone2Name=l_chestStrap_01, swing1LimitDegrees=45, swing2LimitDegrees=45, twistLimitDegrees=45}
 {bone1Name=l_chestStrap_03, bone2Name=l_chestStrap_02, swing1LimitDegrees=45, swing2LimitDegrees=45, twistLimitDegrees=45}
 {bone1Name=l_chestStrap_04, bone2Name=l_chestStrap_03, swing1LimitDegrees=45, swing2LimitDegrees=45, twistLimitDegrees=45}
 {bone1Name=l_chestStrap_05, bone2Name=l_chestStrap_04, swing1LimitDegrees=45, swing2LimitDegrees=45, twistLimitDegrees=45}
 {bone1Name=l_chestStrap_06, bone2Name=l_chestStrap_05, swing1LimitDegrees=45, swing2LimitDegrees=45, twistLimitDegrees=45}
 {bone1Name=r_chestStrap_01, bone2Name=r_chestStrap_Main, swing1LimitDegrees=45, swing2LimitDegrees=45, twistLimitDegrees=45}
 {bone1Name=r_chestStrap_02, bone2Name=r_chestStrap_01, swing1LimitDegrees=45, swing2LimitDegrees=45, twistLimitDegrees=45}
 {bone1Name=r_chestStrap_03, bone2Name=r_chestStrap_02, swing1LimitDegrees=45, swing2LimitDegrees=45, twistLimitDegrees=45}
 {bone1Name=r_chestStrap_04, bone2Name=r_chestStrap_03, swing1LimitDegrees=45, swing2LimitDegrees=45, twistLimitDegrees=45}
 {bone1Name=r_chestStrap_05, bone2Name=r_chestStrap_04, swing1LimitDegrees=45, swing2LimitDegrees=45, twistLimitDegrees=45}
 {bone1Name=r_chestStrap_06, bone2Name=r_chestStrap_05, swing1LimitDegrees=45, swing2LimitDegrees=45, twistLimitDegrees=45}
 {
  bone1Name=thigh_l
  bone2Name=pelvis
  swing1LimitDegrees=45
  swing1Motion=Free
  swing2LimitDegrees=45
  swing2Motion=Free
  twistLimitDegrees=45
  twistMotion=Free
 }
 {
  bone1Name=thigh_r
  bone2Name=pelvis
  swing1LimitDegrees=45
  swing1Motion=Free
  swing2LimitDegrees=45
  swing2Motion=Free
  twistLimitDegrees=45
  twistMotion=Free
 }
 {
  bone1Name=clavicle_l
  bone2Name=spine_05
  swing1LimitDegrees=45
  swing1Motion=Free
  swing2LimitDegrees=45
  swing2Motion=Free
  twistLimitDegrees=45
  twistMotion=Free
 }
 {
  bone1Name=clavicle_r
  bone2Name=spine_05
  swing1LimitDegrees=45
  swing1Motion=Free
  swing2LimitDegrees=45
  swing2Motion=Free
  twistLimitDegrees=45
  twistMotion=Free
 }
 {
  bone1Name=upperarm_l
  bone2Name=clavicle_l
  swing1LimitDegrees=45
  swing1Motion=Free
  swing2LimitDegrees=45
  swing2Motion=Free
  twistLimitDegrees=45
  twistMotion=Free
 }
 {
  bone1Name=upperarm_r
  bone2Name=clavicle_r
  swing1LimitDegrees=45
  swing1Motion=Free
  swing2LimitDegrees=45
  swing2Motion=Free
  twistLimitDegrees=45
  twistMotion=Free
 }
 {
  bone1Name=lowerarm_l
  bone2Name=upperarm_l
  swing1LimitDegrees=45
  swing1Motion=Free
  swing2LimitDegrees=45
  swing2Motion=Free
  twistLimitDegrees=45
  twistMotion=Free
 }
 {
  bone1Name=lowerarm_r
  bone2Name=upperarm_r
  swing1LimitDegrees=45
  swing1Motion=Free
  swing2LimitDegrees=45
  swing2Motion=Free
  twistLimitDegrees=45
  twistMotion=Free
 }
]
