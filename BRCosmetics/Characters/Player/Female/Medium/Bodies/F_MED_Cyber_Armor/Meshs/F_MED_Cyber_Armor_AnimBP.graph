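# Animation Blueprint for the female medium Cyber Armor body cosmetic, following the attached parent's pose and adding secondary motion.
# Copied pose passes through weighted strap and knee constraints, corrective bone offsets and light AnimDynamics on strap symbols, then is cached.
# The cache feeds pelvis-space RigidBody sims, with a separate physics asset and strap tweaks for front end.
# Updates lower acceleration scale during skydiving and compute a gravity override the AnimGraph never reads.
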AnimBlueprint /BRCosmetics/Characters/Player/Female/Medium/Bodies/F_MED_Cyber_Armor/Meshs/F_MED_Cyber_Armor_AnimBP
target_skeleton=&/F_MED_Cyber_Armor_Skeleton
graphs [
 AnimGraph {nodes [
   n1 = "Transform (Modify) Bone - Bone: r_chestStrap_01"!(
    ComponentPose=Local To Component(LocalPose=Use cached pose 'PreSim'())
    bAlphaBoolEnabled:bool=True
    Translation="0,0,0"
    Rotation="0,0,-20"
    Scale="1,1,1"
   ) -> (Pose) {
    alphaInputType=Bool
    rotationMode=BMM_Additive
    scaleMode=BMM_Additive
    translationMode=BMM_Additive
    alphaBoolBlend {blendOption=Linear}
    boneToModify {boneName=r_chestStrap_01}
    rotation {roll=-20}
   }
   n2 = RigidBody!(
    ComponentPose=@n1
    Alpha
    OverrideWorldGravity:Vector="0,0,0"
    ExternalForce
    ComponentLinearAccScale:Vector=Get SkydiveAccScale() -> (SkydiveAccScale:Vector)
   ) -> (Pose) {
    bOverrideWorldGravity=true
    overridePhysicsAsset=&/F_MED_Cyber_Armor_FE_Physics
    simulationSpace=BaseBoneSpace
    alphaBoolBlend {blendOption=Linear}
    baseBoneRef {boneName=pelvis}
    componentAppliedLinearAccClamp=<3000 3000 3000>
    componentLinearAccScale=<0.4 0.4 0.42>
   }
   n3 = "Transform (Modify) Bone - Bone: l_chestStrap_symbol_MSTR"!(ComponentPose=@n2, bAlphaBoolEnabled:bool=True, Translation="0,0,0", Rotation="10,0,0", Scale="1,1,1") -> (Pose) {
    alphaInputType=Bool
    rotationMode=BMM_Additive
    rotationSpace=BCS_BoneSpace
    scaleMode=BMM_Additive
    scaleSpace=BCS_BoneSpace
    translationMode=BMM_Additive
    translationSpace=BCS_BoneSpace
    alphaBoolBlend {blendOption=Linear}
    boneToModify {boneName=l_chestStrap_symbol_MSTR}
    rotation {pitch=10}
   }
   n4 = "Transform (Modify) Bone - Bone: r_chestStrap_symbol_MSTR"!(ComponentPose=@n3, bAlphaBoolEnabled:bool=True, Translation="0,0,0", Rotation="10,0,0", Scale="1,1,1") -> (Pose) {
    alphaInputType=Bool
    rotationMode=BMM_Additive
    rotationSpace=BCS_BoneSpace
    scaleMode=BMM_Additive
    scaleSpace=BCS_BoneSpace
    translationMode=BMM_Additive
    translationSpace=BCS_BoneSpace
    alphaBoolBlend {blendOption=Linear}
    boneToModify {boneName=r_chestStrap_symbol_MSTR}
    rotation {pitch=10}
   }
   n5 = RigidBody(
    ComponentPose=Local To Component(LocalPose=Use cached pose 'PreSim'())
    OverrideWorldGravity="0,0,0"
    ComponentLinearAccScale=Get SkydiveAccScale()
   ) {
    bOverrideWorldGravity=true
    overridePhysicsAsset=&/F_MED_Cyber_Armor_ChestPhysics
    simulationSpace=BaseBoneSpace
    alphaBoolBlend {blendOption=Linear}
    baseBoneRef {boneName=pelvis}
    componentAppliedLinearAccClamp=<3000 3000 3000>
    componentLinearAccScale=<0.4 0.4 0.42>
   }
   n6 = Blend Poses by bool(BlendPose_0=Component To Local(ComponentPose=@n4), BlendPose_1=Component To Local(ComponentPose=@n5), bActiveValue=Get bIsInFrontEnd()) {blendType=Linear}
   Output Pose(Result=Component To Local(ComponentPose=Local To Component(LocalPose=@n6))) {layerGroup=None}
   n7 = "Constraint - l_chestStrap_symbol_MSTR (PPPP)"!(
    ComponentPose=Local To Component(LocalPose=Copy Pose From Mesh() {bCopyCurves=true, bUseAttachedParent=true})
    Alpha=1
    ConstraintWeights_0=0.2522
    ConstraintWeights_1:float=0.2188
    ConstraintWeights_2:float=0.3864
    ConstraintWeights_3:float=0.1427
   ) -> (Pose) {
    alphaBoolBlend {blendOption=Linear}
    boneToModify {boneName=l_chestStrap_symbol_MSTR}
    constraintSetup [
     {targetBone {boneName=spine_04}}
     {targetBone {boneName=spine_05}}
     {targetBone {boneName=clavicle_l}}
     {targetBone {boneName=upperarm_twist_01_l}}
    ]
    constraintWeights [1,1,1,1]
   }
   n8 = "Constraint - r_chestStrap_symbol_MSTR (PPPPPPPPPPPPPP)"!(
    ComponentPose=@n7
    Alpha=1
    ConstraintWeights_0=0
    ConstraintWeights_1:float=0
    ConstraintWeights_2:float=0
    ConstraintWeights_3:float=0.0001
    ConstraintWeights_4:float=0.1821
    ConstraintWeights_5:float=0.2129
    ConstraintWeights_6:float=0
    ConstraintWeights_7:float=0
    ConstraintWeights_8:float=0.4298
    ConstraintWeights_9:float=0
    ConstraintWeights_10:float=0.175
    ConstraintWeights_11:float=0
    ConstraintWeights_12:float=0
    ConstraintWeights_13:float=0
   ) -> (Pose) {
    alphaBoolBlend {blendOption=Linear}
    boneToModify {boneName=r_chestStrap_symbol_MSTR}
    constraintSetup [
     {targetBone {boneName=pelvis}}
     {targetBone {boneName=spine_01}}
     {targetBone {boneName=spine_02}}
     {targetBone {boneName=spine_03}}
     {targetBone {boneName=spine_04}}
     {targetBone {boneName=spine_05}}
     {targetBone {boneName=clavicle_l}}
     {targetBone {boneName=upperarm_twist_01_l}}
     {targetBone {boneName=clavicle_r}}
     {targetBone {boneName=upperarm_r}}
     {targetBone {boneName=upperarm_twist_01_r}}
     {targetBone {boneName=upperarm_twist_02_r}}
     {targetBone {boneName=neck_01}}
     {targetBone {boneName=neck_02}}
    ]
    constraintWeights [1,1,1,1,1,1,1,1,1,1,1,1,1,1]
   }
   n9 = "Constraint - l_chestStrap_Main (PPPPPPPPPPP)"!(
    ComponentPose=@n8
    Alpha=1
    ConstraintWeights_0=0
    ConstraintWeights_1:float=0.1808
    ConstraintWeights_2:float=0
    ConstraintWeights_3:float=0.186
    ConstraintWeights_4:float=0.3447
    ConstraintWeights_5:float=0
    ConstraintWeights_6:float=0.2885
    ConstraintWeights_7:float=0
    ConstraintWeights_8:float=0
    ConstraintWeights_9:float=0
    ConstraintWeights_10:float=0
   ) -> (Pose) {
    alphaBoolBlend {blendOption=Linear}
    boneToModify {boneName=l_chestStrap_Main}
    constraintSetup [
     {targetBone {boneName=spine_02}}
     {targetBone {boneName=spine_04}}
     {targetBone {boneName=spine_03}}
     {targetBone {boneName=spine_05}}
     {targetBone {boneName=clavicle_l}}
     {targetBone {boneName=upperarm_l}}
     {targetBone {boneName=upperarm_twist_01_l}}
     {targetBone {boneName=upperarm_twist_02_l}}
     {targetBone {boneName=clavicle_r}}
     {targetBone {boneName=upperarm_twist_01_r}}
     {targetBone {boneName=neck_02}}
    ]
    constraintWeights [1,1,1,1,1,1,1,1,1,1,1]
   }
   n10 = "Constraint - r_chestStrap_Main (PPPPPPPPPPP)"!(
    ComponentPose=@n9
    Alpha=1
    ConstraintWeights_0=0
    ConstraintWeights_1:float=0.1153
    ConstraintWeights_2:float=0
    ConstraintWeights_3:float=0.1562
    ConstraintWeights_4:float=0
    ConstraintWeights_5:float=0
    ConstraintWeights_6:float=0.3406
    ConstraintWeights_7:float=0.0001
    ConstraintWeights_8:float=0.3877
    ConstraintWeights_9:float=0.0001
    ConstraintWeights_10:float=0
   ) -> (Pose) {
    alphaBoolBlend {blendOption=Linear}
    boneToModify {boneName=r_chestStrap_Main}
    constraintSetup [
     {targetBone {boneName=spine_02}}
     {targetBone {boneName=spine_04}}
     {targetBone {boneName=spine_03}}
     {targetBone {boneName=spine_05}}
     {targetBone {boneName=clavicle_l}}
     {targetBone {boneName=upperarm_twist_01_l}}
     {targetBone {boneName=clavicle_r}}
     {targetBone {boneName=upperarm_r}}
     {targetBone {boneName=upperarm_twist_01_r}}
     {targetBone {boneName=upperarm_twist_02_r}}
     {targetBone {boneName=neck_02}}
    ]
    constraintWeights [1,1,1,1,1,1,1,1,1,1,1]
   }
   n11 = "Constraint - dyn_r_knee (PP)"!(ComponentPose=@n10, bAlphaBoolEnabled:bool=True, ConstraintWeights_0=0, ConstraintWeights_1:float=1) -> (Pose) {
    alphaInputType=Bool
    bAlphaBoolEnabled=false
    alphaBoolBlend {blendInTime=0.1, blendOption=Linear, blendOutTime=0.1}
    boneToModify {boneName=dyn_r_knee}
    constraintSetup [{targetBone {boneName=thigh_twist_01_r}}, {targetBone {boneName=calf_r}}]
    constraintWeights [1,1]
   }
   n12 = "Constraint - dyn_l_knee (PP)"!(ComponentPose=@n11, bAlphaBoolEnabled:bool=True, ConstraintWeights_0=0, ConstraintWeights_1:float=1) -> (Pose) {
    alphaInputType=Bool
    bAlphaBoolEnabled=false
    alphaBoolBlend {blendInTime=0.1, blendOption=Linear, blendOutTime=0.1}
    boneToModify {boneName=dyn_l_knee}
    constraintSetup [{targetBone {boneName=thigh_twist_01_l}}, {targetBone {boneName=calf_l}}]
    constraintWeights [1,1]
   }
   n13 = "Transform (Modify) Bone - Bone: dyn_l_knee"!(ComponentPose=@n12, bAlphaBoolEnabled:bool=True, Translation="0,-1.25,0", Rotation="0,0,0", Scale="1,1,1") -> (Pose) {
    alphaInputType=Bool
    bAlphaBoolEnabled=false
    translationMode=BMM_Additive
    translationSpace=BCS_BoneSpace
    alphaBoolBlend {blendOption=Linear}
    boneToModify {boneName=dyn_l_knee}
    translation {y=-1.25}
   }
   n14 = "Transform (Modify) Bone - Bone: dyn_l_KneeFold"!(ComponentPose=@n13, bAlphaBoolEnabled:bool=True, Translation="-4,0,0", Rotation="0,0,0", Scale="1,1,1") -> (Pose) {
    alphaInputType=Bool
    bAlphaBoolEnabled=false
    translationMode=BMM_Additive
    translationSpace=BCS_BoneSpace
    alphaBoolBlend {blendInTime=0.1, blendOption=Linear, blendOutTime=0.1}
    boneToModify {boneName=dyn_l_KneeFold}
    translation {x=-4}
   }
   n15 = "Transform (Modify) Bone - Bone: dyn_R_KneeFold"!(ComponentPose=@n14, bAlphaBoolEnabled:bool=True, Translation="5,0,0", Rotation="0,0,0", Scale="1,1,1") -> (Pose) {
    alphaInputType=Bool
    bAlphaBoolEnabled=false
    translationMode=BMM_Additive
    translationSpace=BCS_BoneSpace
    alphaBoolBlend {blendInTime=0.1, blendOption=Linear, blendOutTime=0.1}
    boneToModify {boneName=dyn_R_KneeFold}
    translation {x=5}
   }
   n16 = "Transform (Modify) Bone - Bone: dyn_buttHelper"!(ComponentPose=@n15, bAlphaBoolEnabled:bool=True, Translation="3,-4,0", Rotation="0,0,0", Scale="1,1,1") -> (Pose) {
    alphaInputType=Bool
    bAlphaBoolEnabled=false
    scaleMode=BMM_Additive
    translationMode=BMM_Additive
    translationSpace=BCS_BoneSpace
    alphaBoolBlend {blendInTime=0.1, blendOption=Linear, blendOutTime=0.1}
    boneToModify {boneName=dyn_buttHelper}
    translation=<3 -4>
   }
   n17 = Anim Dynamics - Bone: l_chestStrap_symbol(ComponentPose=@n16, Alpha=0.2, ExternalForce="0,0,0") {
    angularDampingOverride=0.9999
    bOverrideAngularDamping=true
    simulationSpace=RootRelative
    alphaBoolBlend {blendOption=Linear}
    boundBone {boneName=l_chestStrap_symbol}
    componentLinearAccScale=<1 1 1>
    componentLinearVelScale=<1 1 1>
    physicsBodyDefinitions [{
      boundBone {boneName=l_chestStrap_symbol}
      boxExtents=<3 10 10>
      constraintSetup {angularLimitsMax=<2 75 45>, angularLimitsMin=<-2 -10 -45>, angularTarget=<0 0 0>}
      localJointOffset=<1.5 -0.05 0.01>
     }]
   }
   n18 = Anim Dynamics - Bone: r_chestStrap_symbol(ComponentPose=@n17, Alpha=0.2, ExternalForce="0,0,0") {
    angularDampingOverride=0.9999
    bOverrideAngularDamping=true
    simulationSpace=RootRelative
    alphaBoolBlend {blendOption=Linear}
    boundBone {boneName=r_chestStrap_symbol}
    componentLinearAccScale=<1 1 1>
    componentLinearVelScale=<1 1 1>
    physicsBodyDefinitions [{
      boundBone {boneName=r_chestStrap_symbol}
      boxExtents=<3 10 10>
      constraintSetup {angularLimitsMax=<2 75 45>, angularLimitsMin=<-2 -10 -45>, angularTarget=<0 0 0>}
      localJointOffset=<1.5 -0.05 0.01>
     }]
   }
   Save cached pose 'PreSim'(Pose=Component To Local(ComponentPose=@n18)) {cacheName=PreSim}
  ]}
 EventGraph {nodes [
   Event BlueprintUpdateAnimation()
   > n1 = Sequence()
   n2 = Branch(execute=@n1.then_1, Condition=OR Boolean(A=Get bIsSkydiveDiveMode(), B=Get bIsSkydiveFloating()))
   > Set SkydiveAccScale(SkydiveAccScale:Vector="0.15,0.15,0.25") -> (Output_Get:Vector)
   Set SkydiveAccScale(execute=@n2.else, SkydiveAccScale="0.4,0.4,0.42")
   n3 = Gravity Override State Function DEPRECATED(
    execute=@n1.then_0
    CharacterPartAnimInstance=Self-Reference()
    GameTimeInSeconds=GetGameTimeInSeconds()
    GravityOverrideParams=Get Gravity Override Params_0() -> (Gravity Override Params_0:Gravity Override Params Struct Structure)
   )
   > Set gravityOverride(gravityOverride:Vector=@n3) -> (Output_Get:Vector)
   TryGetPawnOwner()
  ]}
]
variables [
 Gravity Override Params_0 {
  type=GravityOverrideParamsStruct
  default {skydive_FloatingDown {gravityOverride=<0 -700 200>}, skydive_GoingUp {gravityOverride=<0 -300 -700>}}
 }
 gravityOverride {type=Vector}
 SkydiveAccScale {type=Vector}
]
