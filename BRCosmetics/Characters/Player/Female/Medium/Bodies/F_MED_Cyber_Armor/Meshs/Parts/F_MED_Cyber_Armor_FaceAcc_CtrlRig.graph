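# Control Rig for the female medium Cyber Armor face accessory that re-angles the left and right head feathers when the SlushyJr backpack is worn.
# One "Head geathers fix" function runs Modify Transforms on both headFeather_Main bones with fixed rotations, weighted 0 or 1 by the backpack bool.
# A second pair of rotations follows, weighted by backpack times front-end flag, so it applies only when both variables are true.
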ControlRigBlueprint /BRCosmetics/Characters/Player/Female/Medium/Bodies/F_MED_Cyber_Armor/Meshs/Parts/F_MED_Cyber_Armor_FaceAcc_CtrlRig
graphs [
 New Function_ContainedGraph {nodes [
   n1 = RigVMFunctionEntryNode()
   RigVMFunction_Sequence = &RigVMFunction_Sequence(ExecuteContext=@n1.ExecuteContext)
   n2 = RigVMDispatchNode(False=0, True=1, Condition=RigVMVariableNode() {variable=Is Wearing SlushyJr backpack}.Value) {template="DISPATCH_RigVMDispatch_If(in Condition,in True,in False,out Result)"}
   &RigUnit_ModifyTransforms(
    ItemToModify=((Item=(Type=Bone,Name="l_headFeather_Main"),Transform=(Rotation=(X=-0.234,Y=0,Z=0,W=-0.9722),Translation=(X=0,Y=0,Z=0),Scale3D=(X=1,Y=1,Z=1))))
    WeightMinimum=0
    ExecutePin=@RigVMFunction_Sequence.A
    Weight=@n2.Result
   )
   > &RigUnit_ModifyTransforms(
    ItemToModify=((Item=(Type=Bone,Name="r_headFeather_Main"),Transform=(Rotation=(X=0.2118,Y=0,Z=0,W=-0.9773),Translation=(X=0,Y=0,Z=0),Scale3D=(X=1,Y=1,Z=1))))
    WeightMinimum=0
    Weight=@n2.Result
   )
   Multiply = &RigVMFunction_MathDoubleMul(
    A=@n2.Result
    B=RigVMDispatchNode(False=0, True=1, Condition=RigVMVariableNode() {variable=Is Front End}.Value) {template="DISPATCH_RigVMDispatch_If(in Condition,in True,in False,out Result)"}.Result
   )
   > &RigUnit_ModifyTransforms(
    ItemToModify=((Item=(Type=Bone,Name="l_headFeather_Main"),Transform=(Rotation=(X=-0.0436,Y=0,Z=0,W=0.999),Translation=(X=0,Y=0,Z=0),Scale3D=(X=1,Y=1,Z=1))))
    WeightMinimum=0
    Weight=@Multiply.Result
   )
   > &RigUnit_ModifyTransforms(
    ItemToModify=((Item=(Type=Bone,Name="r_headFeather_Main"),Transform=(Rotation=(X=-0.0175,Y=0,Z=0,W=0.9998),Translation=(X=0,Y=0,Z=0),Scale3D=(X=1,Y=1,Z=1))))
    WeightMinimum=0
    Weight=@Multiply.Result
   )
   RigVMFunctionReturnNode(ExecuteContext=@RigVMFunction_Sequence.B)
  ]}
 RigVMFunctionLibrary {nodes [Head geathers fix = RigVMCollapseNode(ExecuteContext=())]}
 RigVMModel {nodes [
   RigUnit_BeginExecution = &RigUnit_BeginExecution()
   Head geathers fix = RigVMFunctionReferenceNode(ExecuteContext=@RigUnit_BeginExecution.ExecutePin)
  ]}
]
hierarchy [
 root {loc=<0 0 0>, rot=<0 0 0 1>}
 pelvis {parent=root, loc=<0 -0.5301 98.0571>, rot=<0.0127 0.707 -0.0127 -0.707>}
 spine_01 {parent=pelvis, loc=<6.1208 -0.9328 0>, rot=<0 0 0.034 -0.9994>}
 spine_02 {parent=spine_01, loc=<6.5996 -0.2095 0>, rot=<0 0 0.0238 -0.9997>}
 spine_03 {parent=spine_02, loc=<6.4914 -0.1237 0>, rot=<0 0 0.0149 -0.9999>}
 spine_04 {parent=spine_03, loc=<6.9598 0.0794 0>, rot=<0 0 -0.0132 -0.9999>}
 spine_05 {parent=spine_04, loc=<6.9987 0 0>, rot=<0 0 -0.0015 -1>}
 neck_01 {parent=spine_05, loc=<11.3138 0.1529 0>, rot=<0 0 -0.2003 -0.9797>}
 neck_02 {parent=neck_01, loc=<4.5627 0 0>, rot=<0 0 0.0504 -0.9987>}
 head {parent=neck_02, loc=<5.3914 0 0>, rot=<0 0 0.115 -0.9934>}
 r_headFeather_Main {parent=head, loc=<10.5739 0.8954 6.8204>, rot=<0.9052 -0.0798 0.2078 -0.3619>}
 r_headFeather_01 {parent=r_headFeather_Main, loc=<0 0 0>, rot=<0 0 0 1>}
 r_headFeather_02 {parent=r_headFeather_01, loc=<15.2039 0 0>, rot=<-0.016 -0.8035 -0.4088 -0.4326>}
 r_headFeather_03 {parent=r_headFeather_02, loc=<15.8974 0 0>, rot=<-0.2333 -0.2244 0.0746 -0.9432>}
 r_headFeather_04 {parent=r_headFeather_03, loc=<24.3123 0 0>, rot=<-0.0103 0.0772 0.0572 -0.9953>}
 r_headFeather_05 {parent=r_headFeather_04, loc=<31.23 0 0>, rot=<-0.1825 0.0921 0.0024 -0.9789>}
 r_headFeather_06 {parent=r_headFeather_05, loc=<29.171 0 0>, rot=<-0.0037 -0.228 0.0518 -0.9723>}
 r_headFeather_07 {parent=r_headFeather_06, loc=<15.1619 0 0>, rot=<-0.4366 -0.2018 0.1182 -0.8687>}
 l_headFeather_Main {parent=head, loc=<10.5737 0.8954 -6.8204>, rot=<0.3619 -0.2078 0.0798 -0.9052>}
 l_headFeather_01 {parent=l_headFeather_Main, loc=<0 0 0>, rot=<0 0 0 1>}
 l_headFeather_02 {parent=l_headFeather_01, loc=<15.2045 0 0>, rot=<0.016 -0.8035 0.4088 -0.4325>}
 l_headFeather_03 {parent=l_headFeather_02, loc=<15.8977 0 0>, rot=<0.1003 -0.2326 -0.0428 -0.9664>}
 l_headFeather_04 {parent=l_headFeather_03, loc=<24.312 0 0>, rot=<0.0491 0.0615 -0.0738 -0.9942>}
 l_headFeather_05 {parent=l_headFeather_04, loc=<31.2309 0 0>, rot=<0.2793 0.0914 -0.0116 -0.9558>}
 l_headFeather_06 {parent=l_headFeather_05, loc=<29.1705 0 0>, rot=<0.0534 -0.2251 -0.0634 -0.9708>}
 l_headFeather_07 {parent=l_headFeather_06, loc=<15.1619 0 0>, rot=<0.2242 0.1184 0.965 -0.0665>}
 l_sideHair_main {parent=head, loc=<5.4437 1.858 -6.9931>, rot=<-0.0615 -0.9973 -0.0112 0.0384>}
 l_sideHair_01 {parent=l_sideHair_main, loc=<0 0 0>, rot=<0 0 0 1>}
 l_sideHair_02 {parent=l_sideHair_01, loc=<3.7308 0 0>, rot=<0.0326 -0.1116 -0.1022 -0.9879>}
 l_sideHair_03 {parent=l_sideHair_02, loc=<3.2647 0 0>, rot=<-0.0215 -0.0783 -0.0996 -0.9917>}
 r_sideHair_main {parent=head, loc=<6.9233 1.1794 6.7794>, rot=<0.0109 0.1019 0.9891 0.1061>}
 r_sideHair_01 {parent=r_sideHair_main, loc=<0 0 0>, rot=<0 0 0 1>}
 r_sideHair_02 {parent=r_sideHair_01, loc=<4.6902 0 0>, rot=<-0.0092 -0.1139 0.1083 -0.9875>}
 r_sideHair_03 {parent=r_sideHair_02, loc=<3.5104 0 0>, rot=<-0.0624 -0.079 0.0514 -0.9936>}
 l_bangs {parent=head, loc=<9.2472 4.8496 -5.6813>, rot=<-0.1348 -0.9545 -0.2633 -0.0369>}
 c_bangs_01 {parent=head, loc=<10.7046 8.4695 0.5251>, rot=<0.1665 0.6253 0.7491 0.142>}
 c_bangs_02 {parent=c_bangs_01, loc=<3.2729 0 -0.4844>, rot=<0 -0.1368 0 -0.9906>}
 r_bangs {parent=head, loc=<9.6072 4.6488 5.7575>, rot=<0.0616 0.2386 0.9691 0.011>}
 l_hair_main_LookAt {parent=head, loc=<3.6486 -5.7594 -0.1755>, rot=<0.0763 -0.6993 0.7068 -0.0744>}
 l_hair_main {parent=l_hair_main_LookAt, loc=<0 0 0>, rot=<0 0 0 1>}
 l_hair_main_01 {parent=l_hair_main, loc=<0 0 0>, rot=<0 0 0 1>}
 l_hair_main_02 {parent=l_hair_main_01, loc=<10.5145 0.0119 -0.0318>, rot=<-0.0021 0.1279 -0.0011 -0.9918>}
 l_hair_main_03 {parent=l_hair_main_02, loc=<7.3018 0 0>, rot=<-0.0118 -0.1092 0.0047 -0.9939>}
 l_hair_main_04 {parent=l_hair_main_03, loc=<7.3024 0 0>, rot=<0.084 -0.1773 -0.0028 -0.9806>}
 l_hair_main_05 {parent=l_hair_main_04, loc=<8.1629 0 0>, rot=<-0.0506 -0.0538 -0.0027 -0.9973>}
 clavicle_r {parent=spine_05, loc=<5.4831 -0.1053 3.2946>, rot=<0.7289 0.0465 -0.6829 0.009>}
 upperarm_r {parent=clavicle_r, loc=<-9.8658 1.2136 0.4104>, rot=<0.0164 -0.3245 -0.0269 -0.9454>}
 upperarm_twist_01_r {parent=upperarm_r, loc=<-7.2135 0.1134 0>, rot=<0 0 0 1>}
 clavicle_l {parent=spine_05, loc=<5.4831 -0.1053 -3.2946>, rot=<0.0465 -0.7289 -0.009 -0.6829>}
 upperarm_l {parent=clavicle_l, loc=<9.8658 -1.2136 -0.4105>, rot=<0.0164 -0.3245 -0.0269 -0.9454>}
 upperarm_twist_01_l {parent=upperarm_l, loc=<7.2135 -0.1134 0>, rot=<0 0 0 1>}
 thigh_l {parent=pelvis, loc=<-1.0855 -0.2109 -8.7053>, rot=<0.0009 0.0355 0.0155 -0.9993>}
 thigh_twist_01_l {parent=thigh_l, loc=<-27.952 0 0>, rot=<0.0013 0 0 -1>}
 thigh_r {parent=pelvis, loc=<-1.0853 -0.2109 8.7053>, rot=<0.0355 -0.0009 0.9993 0.0155>}
 thigh_twist_01_r {parent=thigh_r, loc=<27.952 0 0>, rot=<0.0013 0 0 -1>}
 dyn_thigh_01_r {parent=thigh_r, loc=<-0.525 0 0>, rot=<0 -1 0 0>}
 dyn_hair_lookat_loc {parent=pelvis, loc=<30.597 -24.2498 0>, rot=<0 0 0 1>}
 r_headFeather_01 {type=Curve, loc=<0 0 0>, rot=<0 0 0 1>}
 r_headFeather_02 {type=Curve, loc=<0 0 0>, rot=<0 0 0 1>}
 r_headFeather_Main {type=Curve, loc=<0 0 0>, rot=<0 0 0 1>}
 r_headFeather_03 {type=Curve, loc=<0 0 0>, rot=<0 0 0 1>}
 r_headFeather_04 {type=Curve, loc=<0 0 0>, rot=<0 0 0 1>}
 r_headFeather_05 {type=Curve, loc=<0 0 0>, rot=<0 0 0 1>}
 r_headFeather_06 {type=Curve, loc=<0 0 0>, rot=<0 0 0 1>}
 r_headFeather_07 {type=Curve, loc=<0 0 0>, rot=<0 0 0 1>}
 l_headFeather_04 {type=Curve, loc=<0 0 0>, rot=<0 0 0 1>}
 l_headFeather_03 {type=Curve, loc=<0 0 0>, rot=<0 0 0 1>}
 l_headFeather_02 {type=Curve, loc=<0 0 0>, rot=<0 0 0 1>}
 l_headFeather_05 {type=Curve, loc=<0 0 0>, rot=<0 0 0 1>}
 l_headFeather_06 {type=Curve, loc=<0 0 0>, rot=<0 0 0 1>}
]
variables [
 Is Front End {type=bool, default=False}
 Is Wearing SlushyJr backpack {type=bool, default=False}
]
