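# Physics asset for the female medium Cyber Armor face accessory, giving its left and right head-feather bone chains physics bodies.
# Each feather chain runs seven capsules from a tiny Main body on the head, with limited swing and twist that loosens toward the tip.
# Capsules on pelvis, spine, neck, head, thighs and upper arms form the body, joined by free constraints.
# A free constraint also links the two first feathers, and spine_05 carries two very long capsules.
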PhysicsAsset /BRCosmetics/Characters/Player/Female/Medium/Bodies/F_MED_Cyber_Armor/Meshs/Parts/F_MED_Cyber_Armor_FaceAcc_Physics
bodies {
 head [{shapeType=Capsule, length=15.3003, radius=7.9763, shapeName=head_capsule, center=<7.098 0 0.0001>}]
 l_headFeather_01 [{shapeType=Capsule, length=4.3374, radius=4.3386, shapeName=l_headFeather_01_capsule, center=<8.8034 -0.0005 -0.0001>, rotation=<rot 90 0 0>}]
 l_headFeather_02 [{shapeType=Capsule, length=11.2499, radius=3.1777, shapeName=l_headFeather_02_capsule, center=<7.9944 -0.0005 0.0002>, rotation=<rot 90 0 0>}]
 l_headFeather_03 [{shapeType=Capsule, length=17.9997, radius=3.5005, shapeName=l_headFeather_03_capsule, center=<11.9019 -0.0005 2>, rotation=<rot 90 0 0>}]
 l_headFeather_04 [{shapeType=Capsule, length=22.7404, radius=4.7396, shapeName=l_headFeather_04_capsule, center=<14.9906 -0.0005 2>, rotation=<rot 90 0 0>}]
 l_headFeather_05 [{shapeType=Capsule, length=21.4043, radius=4.7396, shapeName=l_headFeather_05_capsule, center=<14.4251 -0.0008 0>, rotation=<rot 90 0 0>}]
 l_headFeather_06 [{shapeType=Capsule, length=9.3635, radius=4.7396, shapeName=l_headFeather_06_capsule, center=<7.8374 -0.0005 -0.0001>, rotation=<rot 90 0 0>}]
 l_headFeather_07 [{shapeType=Capsule, length=12.2604, radius=4.7396, shapeName=l_headFeather_07_capsule, center=<-9.3581 -0.0002 0>, rotation=<rot 90 0 0>}]
 l_headFeather_Main [{shapeType=Capsule, length=0.505, radius=0.505, shapeName=l_headFeather_Main_capsule}]
 neck_01 [{shapeType=Capsule, length=14.2574, radius=7.9639, shapeName=neck_01_capsule}]
 neck_02 [{shapeType=Capsule, length=17.6286, radius=9.0169, shapeName=neck_02_capsule}]
 pelvis [{shapeType=Capsule, length=12, radius=16.2199, shapeName=pelvis_capsule}]
 r_headFeather_01 [{shapeType=Capsule, length=4.3374, radius=4.3386, shapeName=r_headFeather_01_capsule, center=<8.8034 -0.0005 -0.0001>, rotation=<rot 90 0 0>}]
 r_headFeather_02 [{shapeType=Capsule, length=11.2499, radius=3.1777, shapeName=r_headFeather_02_capsule, center=<7.9944 -0.0005 0.0002>, rotation=<rot 90 0 0>}]
 r_headFeather_03 [{shapeType=Capsule, length=17.9997, radius=3.5005, shapeName=r_headFeather_03_capsule, center=<11.9019 -0.0005 2>, rotation=<rot 90 0 0>}]
 r_headFeather_04 [{shapeType=Capsule, length=22.7404, radius=4.7396, shapeName=r_headFeather_04_capsule, center=<14.9906 -0.0005 2>, rotation=<rot 90 0 0>}]
 r_headFeather_05 [{shapeType=Capsule, length=21.4043, radius=4.7396, shapeName=r_headFeather_05_capsule, center=<14.4251 -0.0008 0>, rotation=<rot 90 0 0>}]
 r_headFeather_06 [{shapeType=Capsule, length=9.3635, radius=4.7396, shapeName=r_headFeather_06_capsule, center=<7.8374 -0.0005 -0.0001>, rotation=<rot 90 0 0>}]
 r_headFeather_07 [{shapeType=Capsule, length=12.2604, radius=4.7396, shapeName=r_headFeather_07_capsule, center=<9.6791 -0.0005 0>, rotation=<rot 90 0 0>}]
 r_headFeather_Main [{shapeType=Capsule, length=0.505, radius=0.505, shapeName=r_headFeather_Main_capsule}]
 spine_01 [{shapeType=Capsule, length=12, radius=12.6112, shapeName=spine_01_capsule}]
 spine_02 [{shapeType=Capsule, length=12, radius=11.3667, shapeName=spine_02_capsule}]
 spine_03 [{shapeType=Capsule, length=12, radius=8.5933, shapeName=spine_03_capsule}]
 spine_04 [{shapeType=Capsule, length=12, radius=9.5223, shapeName=spine_04_capsule}]
 spine_05 [
  {shapeType=Capsule, length=300, radius=11.5218, shapeName=spine_05_capsule, center=<0 1.0761 0>}
  {shapeType=Capsule, length=300, radius=11.5218, shapeName=spine_05_capsule, center=<15.2046 3.881 0>}
 ]
 thigh_l [{shapeType=Capsule, length=66.6051, radius=17.8513, shapeName=thigh_l_capsule, center=<-46.4501 0.0001 0>, rotation=<rot 90 0 0>}]
 thigh_r [{shapeType=Capsule, length=66.6051, radius=17.8513, shapeName=thigh_l_capsule, center=<43.1848 0.0003 0>, rotation=<rot 90 0 0>}]
 upperarm_l [{shapeType=Capsule, length=24.9824, radius=9.2653, shapeName=upperarm_l_capsule, center=<21.2259 0 0>, rotation=<rot 90 0 0>}]
 upperarm_r [{shapeType=Capsule, length=24.9824, radius=9.2653, shapeName=upperarm_r_capsule, center=<-20.6653 0 0>, rotation=<rot 90 0 0>}]
}
constraints [
 {
  bone1Name=spine_01
  bone2Name=pelvis
  swing1LimitDegrees=45
  swing1Motion=Free
  swing2LimitDegrees=45
  swing2Motion=Free
  twistLimitDegrees=45
  twistMotion=Free
 }
 {
  bone1Name=spine_02
  bone2Name=spine_01
  swing1LimitDegrees=45
  swing1Motion=Free
  swing2LimitDegrees=45
  swing2Motion=Free
  twistLimitDegrees=45
  twistMotion=Free
 }
 {
  bone1Name=spine_03
  bone2Name=spine_02
  swing1LimitDegrees=45
  swing1Motion=Free
  swing2LimitDegrees=45
  swing2Motion=Free
  twistLimitDegrees=45
  twistMotion=Free
 }
 {
  bone1Name=spine_04
  bone2Name=spine_03
  swing1LimitDegrees=45
  swing1Motion=Free
  swing2LimitDegrees=45
  swing2Motion=Free
  twistLimitDegrees=45
  twistMotion=Free
 }
 {
  bone1Name=spine_05
  bone2Name=spine_04
  swing1LimitDegrees=45
  swing1Motion=Free
  swing2LimitDegrees=45
  swing2Motion=Free
  twistLimitDegrees=45
  twistMotion=Free
 }
 {
  bone1Name=neck_01
  bone2Name=spine_05
  swing1LimitDegrees=45
  swing1Motion=Free
  swing2LimitDegrees=45
  swing2Motion=Free
  twistLimitDegrees=45
  twistMotion=Free
 }
 {
  bone1Name=neck_02
  bone2Name=neck_01
  swing1LimitDegrees=45
  swing1Motion=Free
  swing2LimitDegrees=45
  swing2Motion=Free
  twistLimitDegrees=45
  twistMotion=Free
 }
 {
  bone1Name=head
  bone2Name=neck_02
  swing1LimitDegrees=45
  swing1Motion=Free
  swing2LimitDegrees=45
  swing2Motion=Free
  twistLimitDegrees=45
  twistMotion=Free
 }
 {
  bone1Name=r_headFeather_Main
  bone2Name=head
  swing1LimitDegrees=45
  swing1Motion=Free
  swing2LimitDegrees=45
  swing2Motion=Free
  twistLimitDegrees=45
  twistMotion=Free
 }
 {
  bone1Name=l_headFeather_Main
  bone2Name=head
  swing1LimitDegrees=45
  swing1Motion=Free
  swing2LimitDegrees=45
  swing2Motion=Free
  twistLimitDegrees=45
  twistMotion=Free
 }
 {bone1Name=r_headFeather_01, bone2Name=r_headFeather_Main, swing1LimitDegrees=5, swing2LimitDegrees=5, twistLimitDegrees=5}
 {bone1Name=r_headFeather_02, bone2Name=r_headFeather_01, swing1LimitDegrees=10, swing2LimitDegrees=15, twistLimitDegrees=5}
 {bone1Name=r_headFeather_03, bone2Name=r_headFeather_02, swing1LimitDegrees=15, swing2LimitDegrees=15, twistLimitDegrees=5}
 {bone1Name=r_headFeather_04, bone2Name=r_headFeather_03, swing1LimitDegrees=30, swing2LimitDegrees=30, twistLimitDegrees=30}
 {bone1Name=r_headFeather_05, bone2Name=r_headFeather_04, swing1LimitDegrees=30, swing2LimitDegrees=30, twistLimitDegrees=30}
 {bone1Name=r_headFeather_06, bone2Name=r_headFeather_05, swing1LimitDegrees=30, swing2LimitDegrees=30, twistLimitDegrees=30}
 {bone1Name=r_headFeather_07, bone2Name=r_headFeather_06, swing1LimitDegrees=30, swing2LimitDegrees=30, twistLimitDegrees=30}
 {bone1Name=l_headFeather_01, bone2Name=l_headFeather_Main, swing1LimitDegrees=5, swing2LimitDegrees=5, twistLimitDegrees=5}
 {bone1Name=l_headFeather_02, bone2Name=l_headFeather_01, swing1LimitDegrees=10, swing2LimitDegrees=15, twistLimitDegrees=5}
 {bone1Name=l_headFeather_03, bone2Name=l_headFeather_02, swing1LimitDegrees=15, swing2LimitDegrees=15, twistLimitDegrees=5}
 {bone1Name=l_headFeather_04, bone2Name=l_headFeather_03, swing1LimitDegrees=30, swing2LimitDegrees=30, twistLimitDegrees=30}
 {bone1Name=l_headFeather_05, bone2Name=l_headFeather_04, swing1LimitDegrees=30, swing2LimitDegrees=30, twistLimitDegrees=30}
 {bone1Name=l_headFeather_06, bone2Name=l_headFeather_05, swing1LimitDegrees=30, swing2LimitDegrees=30, twistLimitDegrees=30}
 {bone1Name=l_headFeather_07, bone2Name=l_headFeather_06, swing1LimitDegrees=30, swing2LimitDegrees=30, twistLimitDegrees=30}
 {
  bone1Name=thigh_r
  bone2Name=pelvis
  swing1LimitDegrees=45
  swing1Motion=Free
  swing2LimitDegrees=45
  swing2Motion=Free
  twistLimitDegrees=45
  twistMotion=Free
 }
 {
  bone1Name=thigh_l
  bone2Name=pelvis
  swing1LimitDegrees=45
  swing1Motion=Free
  swing2LimitDegrees=45
  swing2Motion=Free
  twistLimitDegrees=45
  twistMotion=Free
 }
 {
  bone1Name=r_headFeather_01
  bone2Name=l_headFeather_01
  swing1LimitDegrees=45
  swing1Motion=Free
  swing2LimitDegrees=45
  swing2Motion=Free
  twistLimitDegrees=45
  twistMotion=Free
 }
 {
  bone1Name=upperarm_r
  bone2Name=spine_05
  swing1LimitDegrees=45
  swing1Motion=Free
  swing2LimitDegrees=45
  swing2Motion=Free
  twistLimitDegrees=45
  twistMotion=Free
 }
 {
  bone1Name=upperarm_l
  bone2Name=spine_05
  swing1LimitDegrees=45
  swing1Motion=Free
  swing2LimitDegrees=45
  swing2Motion=Free
  twistLimitDegrees=45
  twistMotion=Free
 }
]
