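# Physics asset for the short-hair face accessory part of the female medium Cyber Armor character body.
# Capsule bodies cover pelvis, spine, neck, head, clavicles and upper arms, plus a single left main hair chain linked to the head through a LookAt bone.
# Body joints and the hair root links have free swing and twist, while the five links down the strand are the only constraints not set to Free.
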PhysicsAsset /BRCosmetics/Characters/Player/Female/Medium/Bodies/F_MED_Cyber_Armor/Meshs/Parts/F_MED_Cyber_Armor_FaceAcc_ShortHair_Physics
bodies {
 clavicle_l [{shapeType=Capsule, length=17.2697, radius=8.0178, shapeName=clavicle_l_capsule, center=<5.1921 0.3488 0>, rotation=<rot 90 180 176.1569>}]
 clavicle_r [{shapeType=Capsule, length=10.9925, radius=8.2522, shapeName=clavicle_r_capsule, center=<-4.8915 0.9254 0>, rotation=<rot 90 180 -169.2873>}]
 head [{shapeType=Capsule, length=6.1373, radius=6.2934, shapeName=head_capsule}]
 l_hair_main [{shapeType=Capsule, length=0.505, radius=0.505, shapeName=l_hair_main_capsule, rotation=<rot 0 -90 0>}]
 l_hair_main_01 [{shapeType=Capsule, length=8.7159, radius=1.2841, shapeName=l_hair_main_01_capsule, center=<5.0163 0 0>, rotation=<rot 0 0 90>}]
 l_hair_main_02 [{shapeType=Capsule, length=8.4211, radius=1.5789, shapeName=l_hair_main_02_capsule, center=<3.5637 0 0>, rotation=<rot 0 0 90>}]
 l_hair_main_03 [{shapeType=Capsule, length=7.9873, radius=2.0127, shapeName=l_hair_main_03_capsule, center=<3.5995 0 0>, rotation=<rot 0 0 90>}]
 l_hair_main_04 [{shapeType=Capsule, length=7.9873, radius=2.0127, shapeName=l_hair_main_04_capsule, center=<4.3066 0 0>, rotation=<rot 0 0 90>}]
 l_hair_main_05 [{shapeType=Capsule, length=7.3854, radius=2.6146, shapeName=l_hair_main_05_capsule, center=<6.1192 0 0>, rotation=<rot 0 0 90>}]
 l_hair_main_LookAt [{shapeType=Capsule, length=0.505, radius=0.505, shapeName=l_hair_main_LookAt_capsule, rotation=<rot 0 -90 0>}]
 neck_01 [{shapeType=Capsule, length=46.3815, radius=6.0826, shapeName=neck_01_capsule}]
 neck_02 [{shapeType=Capsule, length=46.8209, radius=5.2653, shapeName=neck_02_capsule}]
 pelvis [{shapeType=Capsule, length=20, radius=10, shapeName=pelvis_capsule}]
 spine_01 [{shapeType=Capsule, length=20, radius=10, shapeName=spine_01_capsule}]
 spine_02 [{shapeType=Capsule, length=20, radius=10, shapeName=spine_02_capsule}]
 spine_03 [{shapeType=Capsule, length=21.8608, radius=7.8273, shapeName=spine_03_capsule, rotation=<rot 0 -90 0>}]
 spine_04 [{shapeType=Capsule, length=25.0721, radius=7.1847, shapeName=spine_04_capsule}]
 spine_05 [{shapeType=Capsule, length=24.7791, radius=7.4777, shapeName=spine_05_capsule}]
 upperarm_l [{shapeType=Capsule, length=16.2822, radius=6.8251, shapeName=upperarm_l_capsule, center=<12.2524 0 0>, rotation=<rot 90 0 0>}]
 upperarm_r [{shapeType=Capsule, length=21.978, radius=9.5024, shapeName=upperarm_r_capsule, center=<-11.9845 0 0>, rotation=<rot 90 0 0>}]
}
constraints [
 {
  bone1Name=spine_05
  bone2Name=spine_04
  swing1LimitDegrees=45
  swing1Motion=Free
  swing2LimitDegrees=45
  swing2Motion=Free
  twistLimitDegrees=45
  twistMotion=Free
 }
 {
  bone1Name=spine_04
  bone2Name=spine_03
  swing1LimitDegrees=45
  swing1Motion=Free
  swing2LimitDegrees=45
  swing2Motion=Free
  twistLimitDegrees=45
  twistMotion=Free
 }
 {
  bone1Name=neck_01
  bone2Name=spine_05
  swing1LimitDegrees=45
  swing1Motion=Free
  swing2LimitDegrees=45
  swing2Motion=Free
  twistLimitDegrees=45
  twistMotion=Free
 }
 {
  bone1Name=neck_02
  bone2Name=neck_01
  swing1LimitDegrees=45
  swing1Motion=Free
  swing2LimitDegrees=45
  swing2Motion=Free
  twistLimitDegrees=45
  twistMotion=Free
 }
 {
  bone1Name=head
  bone2Name=neck_02
  swing1LimitDegrees=45
  swing1Motion=Free
  swing2LimitDegrees=45
  swing2Motion=Free
  twistLimitDegrees=45
  twistMotion=Free
 }
 {
  bone1Name=clavicle_l
  bone2Name=spine_05
  swing1LimitDegrees=45
  swing1Motion=Free
  swing2LimitDegrees=45
  swing2Motion=Free
  twistLimitDegrees=45
  twistMotion=Free
 }
 {
  bone1Name=clavicle_r
  bone2Name=spine_05
  swing1LimitDegrees=45
  swing1Motion=Free
  swing2LimitDegrees=45
  swing2Motion=Free
  twistLimitDegrees=45
  twistMotion=Free
 }
 {
  bone1Name=upperarm_r
  bone2Name=clavicle_r
  swing1LimitDegrees=45
  swing1Motion=Free
  swing2LimitDegrees=45
  swing2Motion=Free
  twistLimitDegrees=45
  twistMotion=Free
 }
 {
  bone1Name=upperarm_l
  bone2Name=clavicle_l
  swing1LimitDegrees=45
  swing1Motion=Free
  swing2LimitDegrees=45
  swing2Motion=Free
  twistLimitDegrees=45
  twistMotion=Free
 }
 {
  bone1Name=l_hair_main_LookAt
  bone2Name=head
  swing1LimitDegrees=45
  swing1Motion=Free
  swing2LimitDegrees=45
  swing2Motion=Free
  twistLimitDegrees=45
  twistMotion=Free
 }
 {
  bone1Name=l_hair_main
  bone2Name=l_hair_main_LookAt
  swing1LimitDegrees=45
  swing1Motion=Free
  swing2LimitDegrees=45
  swing2Motion=Free
  twistLimitDegrees=45
  twistMotion=Free
 }
 {bone1Name=l_hair_main_01, bone2Name=l_hair_main, swing1LimitDegrees=45, swing2LimitDegrees=45, twistLimitDegrees=45}
 {bone1Name=l_hair_main_02, bone2Name=l_hair_main_01, swing1LimitDegrees=45, swing2LimitDegrees=45, twistLimitDegrees=45}
 {bone1Name=l_hair_main_03, bone2Name=l_hair_main_02, swing1LimitDegrees=45, swing2LimitDegrees=45, twistLimitDegrees=45}
 {bone1Name=l_hair_main_05, bone2Name=l_hair_main_04, swing1LimitDegrees=45, swing2LimitDegrees=45, twistLimitDegrees=45}
 {bone1Name=l_hair_main_04, bone2Name=l_hair_main_03, swing1LimitDegrees=45, swing2LimitDegrees=45, twistLimitDegrees=45}
 {
  bone1Name=spine_03
  bone2Name=spine_02
  swing1LimitDegrees=45
  swing1Motion=Free
  swing2LimitDegrees=45
  swing2Motion=Free
  twistLimitDegrees=45
  twistMotion=Free
 }
 {
  bone1Name=spine_02
  bone2Name=spine_01
  swing1LimitDegrees=45
  swing1Motion=Free
  swing2LimitDegrees=45
  swing2Motion=Free
  twistLimitDegrees=45
  twistMotion=Free
 }
 {
  bone1Name=spine_01
  bone2Name=pelvis
  swing1LimitDegrees=45
  swing1Motion=Free
  swing2LimitDegrees=45
  swing2Motion=Free
  twistLimitDegrees=45
  twistMotion=Free
 }
]
