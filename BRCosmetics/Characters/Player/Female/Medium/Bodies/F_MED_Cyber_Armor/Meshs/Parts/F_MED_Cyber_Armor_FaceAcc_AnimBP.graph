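# Animation Blueprint for the female Cyber Armor face accessory part, animating its hair and head feathers.
# Copies the attached parent's pose, then selects a backpack path (Control Rig, rigid bodies, weighted bone constraints) or a short-hair RigidBody path with a front-end variant.
# AnimDynamics springs side hair, and bone modifiers pose the feathers.
# The event graph detects back bling and adjusts simulation alpha for back bling and skydiving, plus head-driven hair offset.
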AnimBlueprint /BRCosmetics/Characters/Player/Female/Medium/Bodies/F_MED_Cyber_Armor/Meshs/Parts/F_MED_Cyber_Armor_FaceAcc_AnimBP
target_skeleton=&/F_MED_Cyber_Armor_FaceAcc_Skeleton
graphs [
 AnimGraph {nodes [
   n1 = "Transform (Modify) Bone - Bone: l_headFeather_Main"!(ComponentPose, bAlphaBoolEnabled:bool=True, Translation="0,0,0", Rotation="9.126,4.649,-35.923", Scale="1,1,1") -> (Pose) {
    alphaInputType=Bool
    rotationMode=BMM_Additive
    rotationSpace=BCS_ParentBoneSpace
    translationMode=BMM_Additive
    translationSpace=BCS_ParentBoneSpace
    alphaBoolBlend {blendOption=Linear}
    boneToModify {boneName=l_headFeather_Main}
    rotation=<rot 9.126 4.649 -35.923>
   }
   "Transform (Modify) Bone - Bone: r_headFeather_Main"!(ComponentPose=@n1, bAlphaBoolEnabled:bool=True, Translation="0,0,0", Rotation="-1.606,8.323,10.491", Scale="1,1,1") -> (Pose) {
    alphaInputType=Bool
    rotationMode=BMM_Additive
    rotationSpace=BCS_ParentBoneSpace
    translationMode=BMM_Additive
    translationSpace=BCS_ParentBoneSpace
    alphaBoolBlend {blendOption=Linear}
    boneToModify {boneName=r_headFeather_Main}
    rotation=<rot -1.606 8.323 10.491>
   }
   n2 = RigidBody!(
    ComponentPose=Local To Component(LocalPose=Use cached pose 'ShortHairCache'())
    Alpha=Get rigidBodyAlpha() -> (rigidBodyAlpha:double)
    OverrideWorldGravity:Vector="0,0,0"
    ExternalForce
   ) -> (Pose) {
    bClampLinearTranslationLimitToRefPose=true
    bOverrideWorldGravity=true
    overridePhysicsAsset=&/F_MED_Cyber_Armor_FaceAcc_ShortHair_FE_Physics
    simulationSpace=BaseBoneSpace
    alphaBoolBlend {blendOption=Linear}
    baseBoneRef {boneName=neck_01}
    componentAppliedLinearAccClamp=<1500 1500 1500>
    componentLinearAccScale=<0.4 0.4 0.5>
    componentLinearVelScale=<0.4 0.4 0.5>
   }
   n3 = RigidBody(ComponentPose=Local To Component(LocalPose=Use cached pose 'ShortHairCache'()), Alpha=Get rigidBodyAlpha(), OverrideWorldGravity="0,0,0") {
    bClampLinearTranslationLimitToRefPose=true
    bOverrideWorldGravity=true
    overridePhysicsAsset=&/F_MED_Cyber_Armor_FaceAcc_ShortHair_Physics
    simulationSpace=BaseBoneSpace
    alphaBoolBlend {blendOption=Linear}
    baseBoneRef {boneName=neck_01}
    componentAppliedLinearAccClamp=<1500 1500 1500>
    componentLinearAccScale=<0.4 0.4 0.5>
    componentLinearVelScale=<0.4 0.4 0.5>
   }
   n4 = Blend Poses by bool(BlendPose_0=Component To Local(ComponentPose=@n2), BlendPose_1=Component To Local(ComponentPose=@n3), bActiveValue=Get bIsInFrontEnd()) {blendType=Linear}
   n5 = Control Rig!(Source=Use cached pose 'Pre_Sim'(), Alpha, Is Front End:bool=true, Is Wearing SlushyJr backpack:bool=false) -> (Pose) {
    alphaBoolBlend {blendOption=Linear}
    controlRigAssetReference {blueprintRigClass=&/F_MED_Cyber_Armor_FaceAcc_CtrlRig.F_MED_Cyber_Armor_FaceAcc_CtrlRig_C}
   }
   n6 = RigidBody(ComponentPose=Local To Component(LocalPose=@n5), Alpha=Get rigidBodyAlpha(), OverrideWorldGravity="0,0,-980") {
    bClampLinearTranslationLimitToRefPose=true
    bOverrideWorldGravity=true
    overridePhysicsAsset=&/F_MED_Cyber_Armor_FaceAcc_Backpack_Physics
    simulationSpace=BaseBoneSpace
    alphaBoolBlend {blendOption=Linear}
    baseBoneRef {boneName=spine_02}
    componentAppliedLinearAccClamp=<3000 3000 3000>
    componentLinearAccScale=<0.45 0.45 0.6>
    componentLinearVelScale=<0.45 0.45 0.6>
    overrideWorldGravity {z=-980}
   }
   n7 = RigidBody(ComponentPose=@n6, Alpha=Get rigidBodyAlpha(), OverrideWorldGravity="0,0,-980") {
    bClampLinearTranslationLimitToRefPose=true
    bOverrideWorldGravity=true
    overridePhysicsAsset=&F_MED_Cyber_Armor_FaceAcc_ShortHair_Physics
    simulationSpace=BaseBoneSpace
    alphaBoolBlend {blendOption=Linear}
    baseBoneRef {boneName=spine_05}
    componentAppliedLinearAccClamp=<1000 1000 1000>
    overrideWorldGravity {z=-980}
   }
   n8 = "Transform (Modify) Bone - Bone: l_headFeather_01"!(ComponentPose=@n7, bAlphaBoolEnabled:bool=True, Translation="0,0,0", Rotation="0,0,5", Scale="1,1,1") -> (Pose) {
    alphaInputType=Bool
    bAlphaBoolEnabled=false
    rotationMode=BMM_Additive
    translationMode=BMM_Additive
    alphaBoolBlend {blendInTime=0.1, blendOption=Linear, blendOutTime=0.1}
    boneToModify {boneName=l_headFeather_01}
    rotation {roll=5}
   }
   n9 = "Transform (Modify) Bone - Bone: r_headFeather_01"!(ComponentPose=@n8, bAlphaBoolEnabled:bool=True, Translation="0,0,0", Rotation="0,0,5", Scale="1,1,1") -> (Pose) {
    alphaInputType=Bool
    bAlphaBoolEnabled=false
    rotationMode=BMM_Additive
    translationMode=BMM_Additive
    alphaBoolBlend {blendInTime=0.1, blendOption=Linear, blendOutTime=0.1}
    boneToModify {boneName=r_headFeather_01}
    rotation {roll=5}
   }
   n10 = "Constraint - l_hair_main_02 (PPPPP)"!(
    ComponentPose=@n9
    Alpha=1
    ConstraintWeights_0=0.0313
    ConstraintWeights_1:float=0.5114
    ConstraintWeights_2:float=0.187
    ConstraintWeights_3:float=0.2647
    ConstraintWeights_4:float=0.0056
   ) -> (Pose) {
    alphaBoolBlend {blendInTime=0.1, blendOption=Linear, blendOutTime=0.1}
    boneToModify {boneName=l_hair_main_02}
    constraintSetup [
     {targetBone {boneName=spine_03}}
     {targetBone {boneName=spine_05}}
     {targetBone {boneName=neck_01}}
     {targetBone {boneName=neck_02}}
     {targetBone {boneName=head}}
    ]
    constraintWeights [1,1,1,1,1]
   }
   n11 = "Constraint - l_hair_main_03 (PPPPPPPPPPP)"!(
    ComponentPose=@n10
    Alpha=1
    ConstraintWeights_0=0
    ConstraintWeights_1:float=0.0011
    ConstraintWeights_2:float=0.0019
    ConstraintWeights_3:float=0.0156
    ConstraintWeights_4:float=0.6438
    ConstraintWeights_5:float=0.1662
    ConstraintWeights_6:float=0.0045
    ConstraintWeights_7:float=0.1572
    ConstraintWeights_8:float=0.0098
    ConstraintWeights_9:float=0
    ConstraintWeights_10:float=0.0001
   ) -> (Pose) {
    alphaBoolBlend {blendInTime=0.1, blendOption=Linear, blendOutTime=0.1}
    boneToModify {boneName=l_hair_main_03}
    constraintSetup [
     {targetBone {boneName=spine_01}}
     {targetBone {boneName=spine_02}}
     {targetBone {boneName=spine_03}}
     {targetBone {boneName=spine_04}}
     {targetBone {boneName=spine_05}}
     {targetBone {boneName=clavicle_l}}
     {targetBone {boneName=upperarm_twist_01_l}}
     {targetBone {boneName=clavicle_r}}
     {targetBone {boneName=upperarm_twist_01_r}}
     {targetBone {boneName=neck_01}}
     {targetBone {boneName=neck_02}}
    ]
    constraintWeights [1,1,1,1,1,1,1,1,1,1,1]
   }
   n12 = "Constraint - l_hair_main_04 (PPPPPPPP)"!(
    ComponentPose=@n11
    Alpha=1
    ConstraintWeights_0=0.0048
    ConstraintWeights_1:float=0.0083
    ConstraintWeights_2:float=0.0591
    ConstraintWeights_3:float=0.4558
    ConstraintWeights_4:float=0.2328
    ConstraintWeights_5:float=0.0072
    ConstraintWeights_6:float=0.2237
    ConstraintWeights_7:float=0.0083
   ) -> (Pose) {
    alphaBoolBlend {blendInTime=0.1, blendOption=Linear, blendOutTime=0.1}
    boneToModify {boneName=l_hair_main_04}
    constraintSetup [
     {targetBone {boneName=spine_02}}
     {targetBone {boneName=spine_03}}
     {targetBone {boneName=spine_04}}
     {targetBone {boneName=spine_05}}
     {targetBone {boneName=clavicle_l}}
     {targetBone {boneName=upperarm_twist_01_l}}
     {targetBone {boneName=clavicle_r}}
     {targetBone {boneName=upperarm_twist_01_r}}
    ]
    constraintWeights [1,1,1,1,1,1,1,1]
   }
   n13 = "Constraint - l_hair_main_05 (PPPPPPPP)"!(
    ComponentPose=@n12
    Alpha=1
    ConstraintWeights_0=0.0263
    ConstraintWeights_1:float=0.0757
    ConstraintWeights_2:float=0.1102
    ConstraintWeights_3:float=0.3871
    ConstraintWeights_4:float=0.1914
    ConstraintWeights_5:float=0.0993
    ConstraintWeights_6:float=0.0956
    ConstraintWeights_7:float=0.0145
   ) -> (Pose) {
    alphaBoolBlend {blendInTime=0.1, blendOption=Linear, blendOutTime=0.1}
    boneToModify {boneName=l_hair_main_05}
    constraintSetup [
     {targetBone {boneName=spine_01}}
     {targetBone {boneName=spine_02}}
     {targetBone {boneName=spine_03}}
     {targetBone {boneName=spine_04}}
     {targetBone {boneName=spine_05}}
     {targetBone {boneName=clavicle_l}}
     {targetBone {boneName=clavicle_r}}
     {targetBone {boneName=upperarm_twist_01_r}}
    ]
    constraintWeights [1,1,1,1,1,1,1,1]
   }
   n14 = Component To Local(ComponentPose=@n13)
   n15 = Get isWearingBackpack() -> (isWearingBackpack:bool)
   n16 = Anim Dynamics - Bone: l_sideHair_01!(
    ComponentPose=Local To Component(LocalPose=Blend Poses by bool(BlendPose_0=@n14, BlendPose_1=@n4, bActiveValue=@n15) {blendType=Linear})
    Alpha=0.6
    GravityOverride:Vector="0,0,0"
    ExternalForce="0,0,0"
   ) -> (Pose) {
    angularDampingOverride=1
    angularSpringConstant=550
    bAngularSpring=true
    bGravityOverrideInSimSpace=true
    bOverrideAngularDamping=true
    bUseGravityOverride=true
    simulationSpace=RootRelative
    alphaBoolBlend {blendOption=Linear}
    boundBone {boneName=l_sideHair_01}
    physicsBodyDefinitions [{
      boundBone {boneName=l_sideHair_01}
      boxExtents=<5 10 10>
      constraintSetup {angularLimitsMax=<2 0 45>, angularLimitsMin=<-2 0 -20>, angularTarget=<0 0 0>}
      localJointOffset=<1.5 -0.05 0.1>
     }]
   }
   n17 = Anim Dynamics - Bone: l_sideHair_02!(ComponentPose=@n16, Alpha=1, GravityOverride:Vector="0,0,0", ExternalForce="0,0,0") -> (Pose) {
    angularDampingOverride=1
    angularSpringConstant=550
    bAngularSpring=true
    bGravityOverrideInSimSpace=true
    bOverrideAngularDamping=true
    bUseGravityOverride=true
    simulationSpace=RootRelative
    alphaBoolBlend {blendOption=Linear}
    boundBone {boneName=l_sideHair_02}
    physicsBodyDefinitions [{
      boundBone {boneName=l_sideHair_01}
      boxExtents=<5 10 10>
      constraintSetup {angularLimitsMax=<2 0 45>, angularLimitsMin=<-2 0 -45>, angularTarget=<0 0 0>}
      localJointOffset=<1.5 -0.05 0.1>
     }]
   }
   n18 = Anim Dynamics - Bone: l_sideHair_03!(ComponentPose=@n17, Alpha=1, GravityOverride:Vector="0,0,0", ExternalForce="0,0,0") -> (Pose) {
    angularDampingOverride=1
    angularSpringConstant=550
    bAngularSpring=true
    bGravityOverrideInSimSpace=true
    bOverrideAngularDamping=true
    bUseGravityOverride=true
    simulationSpace=RootRelative
    alphaBoolBlend {blendOption=Linear}
    boundBone {boneName=l_sideHair_03}
    physicsBodyDefinitions [{
      boundBone {boneName=l_sideHair_01}
      boxExtents=<5 10 10>
      constraintSetup {angularLimitsMax=<2 0 80>, angularLimitsMin=<-2 0 -80>, angularTarget=<0 0 0>}
      localJointOffset=<1.5 -0.05 0.1>
     }]
   }
   n19 = Anim Dynamics - Bone: r_sideHair_01!(ComponentPose=@n18, Alpha=0.4, GravityOverride:Vector="0,0,0", ExternalForce="0,0,0") -> (Pose) {
    angularDampingOverride=1
    angularSpringConstant=550
    bAngularSpring=true
    bGravityOverrideInSimSpace=true
    bOverrideAngularDamping=true
    bUseGravityOverride=true
    simulationSpace=RootRelative
    alphaBoolBlend {blendOption=Linear}
    boundBone {boneName=r_sideHair_01}
    physicsBodyDefinitions [{
      boundBone {boneName=r_sideHair_01}
      boxExtents=<5 10 10>
      constraintSetup {angularLimitsMax=<2 0 45>, angularLimitsMin=<-2 0 -20>, angularTarget=<0 0 0>}
      localJointOffset=<1.5 -0.05 0.1>
     }]
   }
   n20 = Anim Dynamics - Bone: r_sideHair_02!(ComponentPose=@n19, Alpha=0.8, GravityOverride:Vector="0,0,0", ExternalForce="0,0,0") -> (Pose) {
    angularDampingOverride=1
    angularSpringConstant=550
    bAngularSpring=true
    bGravityOverrideInSimSpace=true
    bOverrideAngularDamping=true
    bUseGravityOverride=true
    simulationSpace=RootRelative
    alphaBoolBlend {blendOption=Linear}
    boundBone {boneName=r_sideHair_02}
    physicsBodyDefinitions [{
      boundBone {boneName=r_sideHair_02}
      boxExtents=<5 10 10>
      constraintSetup {angularLimitsMax=<2 0 45>, angularLimitsMin=<-2 0 -45>, angularTarget=<0 0 0>}
      localJointOffset=<1.5 -0.05 0.1>
     }]
   }
   n21 = Anim Dynamics - Bone: r_sideHair_03!(ComponentPose=@n20, Alpha=0.8, GravityOverride:Vector="0,0,0", ExternalForce="0,0,0") -> (Pose) {
    angularDampingOverride=1
    angularSpringConstant=550
    bAngularSpring=true
    bGravityOverrideInSimSpace=true
    bOverrideAngularDamping=true
    bUseGravityOverride=true
    simulationSpace=RootRelative
    alphaBoolBlend {blendOption=Linear}
    boundBone {boneName=r_sideHair_03}
    physicsBodyDefinitions [{
      boundBone {boneName=r_sideHair_03}
      boxExtents=<5 10 10>
      constraintSetup {angularLimitsMax=<2 0 80>, angularLimitsMin=<-2 0 -80>, angularTarget=<0 0 0>}
      localJointOffset=<1.5 -0.05 0.1>
     }]
   }
   n22 = "Constraint - r_headFeather_01 (R)"!(ComponentPose=@n21, bAlphaBoolEnabled:bool=True, ConstraintWeights_0=1) -> (Pose) {
    alphaInputType=Bool
    bAlphaBoolEnabled=false
    alphaBoolBlend {blendInTime=0.1, blendOption=Linear, blendOutTime=0.1}
    boneToModify {boneName=r_headFeather_01}
    constraintSetup [{transformType=Rotation, targetBone {boneName=l_headFeather_01}}]
    constraintWeights [1]
   }
   n23 = "Constraint - l_headFeather_01 (R)"!(ComponentPose=@n22, bAlphaBoolEnabled:bool=True, ConstraintWeights_0=1) -> (Pose) {
    alphaInputType=Bool
    bAlphaBoolEnabled=false
    alphaBoolBlend {blendInTime=0.1, blendOption=Linear, blendOutTime=0.1}
    boneToModify {boneName=l_headFeather_01}
    constraintSetup [{transformType=Rotation, targetBone {boneName=r_headFeather_01}}]
    constraintWeights [1]
   }
   n24 = "Get Rotation X (Roll)"() -> ("Rotation X (Roll)":double)
   n25 = "Transform (Modify) Bone - Bone: l_headFeather_01"!(
    ComponentPose=@n23
    bAlphaBoolEnabled=True
    Translation="0,0,0"
    Rotation_Roll:float=@n24
    Rotation_Pitch:float=-5
    Rotation_Yaw:float=0
    Scale="1,1,1"
   ) -> (Pose) {
    alphaInputType=Bool
    bAlphaBoolEnabled=false
    rotationMode=BMM_Additive
    translationMode=BMM_Additive
    alphaBoolBlend {blendInTime=0.1, blendOption=Linear, blendOutTime=0.1}
    boneToModify {boneName=l_headFeather_01}
    rotation {pitch=-5, roll=50}
   }
   n26 = "Transform (Modify) Bone - Bone: r_headFeather_01"!(
    ComponentPose=@n25
    bAlphaBoolEnabled=True
    Translation="0,0,0"
    Rotation_Roll:float=@n24
    Rotation_Pitch:float=5
    Rotation_Yaw:float=0
    Scale="1,1,1"
   ) -> (Pose) {
    alphaInputType=Bool
    bAlphaBoolEnabled=false
    rotationMode=BMM_Additive
    translationMode=BMM_Additive
    alphaBoolBlend {blendInTime=0.1, blendOption=Linear, blendOutTime=0.1}
    boneToModify {boneName=r_headFeather_01}
    rotation {pitch=5, roll=50}
   }
   n27 = "Transform (Modify) Bone - Bone: l_headFeather_02"!(ComponentPose=@n26, bAlphaBoolEnabled:bool=True, Translation="0,0,0", Rotation="0,-30,0", Scale="1,1,1") -> (Pose) {
    alphaInputType=Bool
    translationMode=BMM_Additive
    translationSpace=BCS_BoneSpace
    alphaBoolBlend {blendInTime=0.1, blendOption=Linear, blendOutTime=0.1}
    boneToModify {boneName=l_headFeather_02}
    rotation {yaw=-30}
   }
   n28 = "Transform (Modify) Bone - Bone: r_headFeather_02"!(ComponentPose=@n27, bAlphaBoolEnabled:bool=True, Translation="0,0,0", Rotation="0,-30,0", Scale="1,1,1") -> (Pose) {
    alphaInputType=Bool
    translationMode=BMM_Additive
    translationSpace=BCS_BoneSpace
    alphaBoolBlend {blendInTime=0.1, blendOption=Linear, blendOutTime=0.1}
    boneToModify {boneName=r_headFeather_02}
    rotation {yaw=-30}
   }
   n29 = "Constraint - l_headFeather_01 (PPPPP)"!(
    ComponentPose=@n28
    bAlphaBoolEnabled:bool=True
    ConstraintWeights_0=0.9309
    ConstraintWeights_1:float=0.0004
    ConstraintWeights_2:float=0.0536
    ConstraintWeights_3:float=0.0148
    ConstraintWeights_4:float=0.0002
   ) -> (Pose) {
    alphaInputType=Bool
    alphaBoolBlend {blendOption=Linear}
    boneToModify {boneName=l_headFeather_01}
    constraintSetup [
     {targetBone {boneName=head}}
     {targetBone {boneName=l_headFeather_Main}}
     {targetBone {boneName=l_headFeather_01}}
     {targetBone {boneName=l_headFeather_02}}
     {targetBone {boneName=l_headFeather_03}}
    ]
    constraintWeights [1,1,1,1,1]
   }
   n30 = "Constraint - r_headFeather_01 (PPPPP)"!(
    ComponentPose=@n29
    bAlphaBoolEnabled:bool=True
    ConstraintWeights_0=0.9081
    ConstraintWeights_1:float=0.0124
    ConstraintWeights_2:float=0.0595
    ConstraintWeights_3:float=0.0194
    ConstraintWeights_4:float=0.0005
   ) -> (Pose) {
    alphaInputType=Bool
    alphaBoolBlend {blendOption=Linear}
    boneToModify {boneName=r_headFeather_01}
    constraintSetup [
     {targetBone {boneName=head}}
     {targetBone {boneName=r_headFeather_Main}}
     {targetBone {boneName=r_headFeather_01}}
     {targetBone {boneName=r_headFeather_02}}
     {targetBone {boneName=r_headFeather_03}}
    ]
    constraintWeights [1,1,1,1,1]
   }
   n31 = "Transform (Modify) Bone - Bone: l_headFeather_01"!(ComponentPose=@n30, bAlphaBoolEnabled=True, Translation="0,0,0", Rotation:Rotator="-10,0,35", Scale="1,1,1") -> (Pose) {
    alphaInputType=Bool
    rotationMode=BMM_Additive
    translationMode=BMM_Additive
    translationSpace=BCS_BoneSpace
    alphaBoolBlend {blendInTime=0.1, blendOption=Linear, blendOutTime=0.1}
    boneToModify {boneName=l_headFeather_01}
    rotation {pitch=-10, roll=35}
   }
   n32 = "Transform (Modify) Bone - Bone: r_headFeather_01"!(ComponentPose=@n31, bAlphaBoolEnabled=True, Translation="0,0,0", Rotation:Rotator="-10,0,35", Scale="1,1,1") -> (Pose) {
    alphaInputType=Bool
    rotationMode=BMM_Additive
    translationMode=BMM_Additive
    translationSpace=BCS_BoneSpace
    alphaBoolBlend {blendInTime=0.1, blendOption=Linear, blendOutTime=0.1}
    boneToModify {boneName=r_headFeather_01}
    rotation {pitch=-10, roll=35}
   }
   Output Pose(Result=Component To Local(ComponentPose=@n32)) {layerGroup=None}
   n33 = Layered blend per bone(
    BasePose=Copy Pose From Mesh() {bCopyCurves=true, bUseAttachedParent=true}
    BlendPoses_0=Copy Pose From Mesh(SourceMeshComponent=Get Body_Mesh())
   ) {layerSetup [{branchFilters [{boneName=thigh_l}, {boneName=thigh_r}, {boneName=calf_l}, {boneName=calf_r}]}]}
   n34 = "Transform (Modify) Bone - Bone: dyn_hair_lookat_loc"(ComponentPose=Local To Component(LocalPose=@n33), Alpha=1, Translation="0,5,0", Rotation="0,0,0", Scale="1,1,1") {
    translationMode=BMM_Additive
    translationSpace=BCS_BoneSpace
    alphaBoolBlend {blendOption=Linear}
    boneToModify {boneName=dyn_hair_lookat_loc}
    translation {y=5}
   }
   n35 = Look At - Bone: l_hair_main_LookAt(ComponentPose=@n34, Alpha=1) {
    alphaBoolBlend {blendOption=Linear}
    boneToModify {boneName=l_hair_main_LookAt}
    lookAtLocation {x=0}
    lookAtTarget {boneReference {boneName=dyn_hair_lookat_loc}}
    lookAt_Axis {axis=<1 0>}
   }
   Save cached pose 'Pre_Sim'(Pose=Component To Local(ComponentPose=@n35)) {cacheName=Pre_Sim}
   n36 = "Transform (Modify) Bone - Bone: l_hair_main"(
    ComponentPose=Local To Component(LocalPose=Use cached pose 'Pre_Sim'())
    Alpha=Get headUp() -> (headUp:double)
    Translation="2,0,2"
    Rotation="0,0,30"
    Scale="1,1,1"
   ) {
    rotationMode=BMM_Additive
    rotationSpace=BCS_BoneSpace
    translationMode=BMM_Additive
    translationSpace=BCS_BoneSpace
    alphaBoolBlend {blendOption=Linear}
    alphaScaleBiasClamp {bMapRange=true, inRange {max=0, min=-50}, outRange {max=0, min=1}}
    boneToModify {boneName=l_hair_main}
    rotation {roll=30}
    translation {x=2, z=2}
   }
   n37 = RigidBody(ComponentPose=@n36, Alpha=Get rigidBodyAlpha(), OverrideWorldGravity="0,0,-980") {
    bClampLinearTranslationLimitToRefPose=true
    bOverrideWorldGravity=true
    overridePhysicsAsset=&/F_MED_Cyber_Armor_FaceAcc_Physics
    simulationSpace=BaseBoneSpace
    alphaBoolBlend {blendOption=Linear}
    baseBoneRef {boneName=spine_02}
    componentAppliedLinearAccClamp=<3000 3000 3000>
    componentLinearAccScale=<0.55 0.55 0.65>
    componentLinearVelScale=<0.55 0.55 0.55>
    overrideWorldGravity {z=-980}
   }
   Save cached pose 'ShortHairCache'(Pose=Component To Local(ComponentPose=@n37)) {cacheName=ShortHairCache}
  ]}
 EventGraph {nodes [
   Event BlueprintInitializeAnimation()
   > Set Body_Mesh(Body_Mesh=GetOwnerSkeletalMeshComponent())
   > Cast To PlayerPawn_Athena(Object=TryGetPawnOwner())
   n1 = Event OnCharacterCustomizationLoaded()
   n2 = CheckForBackbling()
   n3 = OR Boolean!(A=@n2, B=@n2.bIsBackblingACape, C:bool=@n2.bIsBackblingLong) -> (ReturnValue)
   > Set isWearingBackpack(isWearingBackpack:bool=@n3) -> (Output_Get:bool)
   > Set Skel mesh backpack(Skel mesh backpack:Skeletal Mesh Component Budgeted Object Reference=Get Array Item with EFort Custom Part Type(Array=@n1.CustomMeshParts, PartType=Backpack)) -> (Output_Get:Skeletal Mesh Component Budgeted Object Reference)
   n4 = MatchesWildcard(
    SourceString=GetDisplayName(Object=Get Skel mesh backpack() -> (Skel mesh backpack:Skeletal Mesh Component Budgeted Object Reference))
    Wildcard=M_MED_Snowman
    SearchCase=CaseSensitive
   )
   > Set Is wearing slushyJr backpack(Is wearing slushyJr backpack:bool=AND Boolean(A=@n3, B=NOT Boolean(A=@n4))) -> (Output_Get:bool)
   Event BlueprintUpdateAnimation()
   > n5 = Sequence!(execute) -> (then_0, then_1, then_2:Exec, then_3:Exec)
   > n6 = "Gravity Override State Function (Parameter Asset) (Nativized)"(
    CharacterPartAnimInstance=Self-Reference()
    GameTimeInSeconds=GetGameTimeInSeconds()
    DebugJointName=spine_05
    Parameters=&/F_MED_Cyber_Armor_FaceAcc_Parameters
   )
   > Set GravityOverride(GravityOverride=@n6)
   n7 = Get IsLongBackblingOn() -> (IsLongBackblingOn:bool)
   n8 = Branch(
    execute=[@n5.then_1, @n17.OutputPin]
    Condition=AND Boolean(
     A=Get IsBackblingOn() -> (IsBackblingOn:bool)
     B="Not Equal (Boolean)"!(A:bool=@n7, B:bool=true) -> (ReturnValue:bool) {is=title K2Node_CallFunction}
    )
   )
   > Set rigidBodyAlpha(rigidBodyAlpha:double=0) -> (Output_Get:double)
   > n9 = Sequence!(execute) -> (then_0, then_1)
   > Set skyDiveFloatPoseFixAlpha(skyDiveFloatPoseFixAlpha:double=0) -> (Output_Get:double)
   n10 = Branch(execute=@n9.then_1, Condition=Get isBistroGliderBool() -> (isBistroGliderBool:bool))
   > Set skyDiveShortBackBling(skyDiveShortBackBling:double=0) -> (Output_Get:double)
   > Set rigidBodyAlpha(rigidBodyAlpha=0.6)
   Set skyDiveShortBackBling(execute=@n10.else, skyDiveShortBackBling=1)
   n11 = Sequence(execute=@n8.else)
   > n12 = Branch(Condition=Get bIsSkyDiving())
   > Set rigidBodyAlpha(rigidBodyAlpha=0.6)
   Set rigidBodyAlpha(execute=@n12.else, rigidBodyAlpha=1)
   n13 = Branch(execute=@n11.then_1, Condition=OR Boolean!(A=Get bIsSkydiveDiveMode(), B=Get bIsSkydiveFloating()) -> (ReturnValue))
   > Set skyDiveFloatPoseFixAlpha(skyDiveFloatPoseFixAlpha=1)
   Set skyDiveFloatPoseFixAlpha(execute=@n13.else, skyDiveFloatPoseFixAlpha=0)
   n14 = GetDeltaTransformFromRefPose(self=GetOwnerSkeletalMeshComponent(), BoneName=head, BaseName=spine_05)
   Set headUp(execute=[@n16.else, @n5.then_2], headUp:double=BreakRotator(InRot=BreakTransform(InTransform=@n14).Rotation).Yaw) -> (Output_Get:double)
   > Set Arm R rot(Arm R rot:double=GetAlphaFromReferenceTransformDelta(SkeletalMeshComponent=Get Body_Mesh(), BoneName=upperarm_r, BaseName=spine_05, InputRange=(X=60,Y=120), TwistAxis="1,0,0")) -> (Output_Get:double)
   n15 = Branch(execute=@n5.then_3, Condition=Get isWearingBackpack())
   > "Set Rotation X (Roll)"("Rotation X (Roll)":double=30) -> (Output_Get:double)
   "Set Rotation X (Roll)"(execute=@n15.else, "Rotation X (Roll)"=50)
   n16 = Branch(Condition=Get bIsSkyDiving())
   > Set headUp(headUp=0)
   n17 = Reroute Node()
  ]}
]
variables [
 Arm R rot {type=number}
 Body_Mesh {type=SkeletalMeshComponent}
 gravityOverride {type=Vector}
 headUp {type=number}
 Is wearing slushyJr backpack {type=boolean}
 IsBackblingOn {type=boolean}
 isBistroGliderBool {type=boolean}
 IsLongBackblingOn {type=boolean}
 isWearingBackpack {type=boolean}
 rigidBodyAlpha {type=number}
 "Rotation X (Roll)" {type=number, default=50}
 Skel mesh backpack {type=SkeletalMeshComponentBudgeted}
 skyDiveFloatPoseFixAlpha {type=number}
 skyDiveShortBackBling {type=number}
]
